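# Default quality-level asphalt layer material function for Helios landscapes, outputting material attributes and a crack-height blend mask.
# Asphalt detail and road-crack textures are sampled through world-space UVs at fixed scales.
# Squared LayerWeight times an inverted variation mask drives a HeightLerp, whose alpha blends normals and displacement between crack and asphalt sources.
# Normals pass through world-aligned conversion with optional tangent-space transform, and a parameter scales displacement from neutral.
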MaterialFunction /Game/Athena/Helios/Environments/Landscape/Materials/QualityLevelLayers/MF_Helios_Asphalt_Default
graph {nodes [
  n1 = MaterialFunctionCall!(Scale=NamedRerouteUsage()) {materialFunction=/Game/Athena/Apollo/Environments/Landscape/Materials/Biome/CommonMaterials/MF_SimpleWorldUV}
  n2 = TextureSample(UVs=@n1.Result) {
   samplerSource=SSM_Wrap_WorldGroupSettings
   samplerType=SAMPLERTYPE_Masks
   texture=/Game/Athena/Apollo/Environments/Landscape/Textures/Roads/T_AsphaltDetail1_S
  }
  n3 = MaterialFunctionCall(Float3=@n2.RGB) {materialFunction=&BreakOutFloat3Components}
  n4 = NamedRerouteUsage()
  n5 = NamedRerouteUsage()
  n6 = FunctionInput(Preview=Constant()) {bUsePreviewValueAsDefault=true, inputName=LayerWeight, inputType=FunctionInput_Scalar, sortPriority=-500}
  n7 = MaterialFunctionCall!(Coordinates, R_01_Scale, R_02_Scale, R_03_Scale, R_04_Scale, G_01_Scale, G_02_Scale, G_03_Scale, G_04_Scale) {materialFunction=/Game/Athena/Helios/Environments/Landscape/Materials/MaterialFunctions_Utilities/MF_VariationMasks2}
  n8 = MaterialFunctionCall(Transition Phase=Multiply(A=Multiply(A=@n6, B=@n6), B=OneMinus(None=@n7.R_01)), Height Texture=NamedRerouteUsage()) {materialFunction=&HeightLerp}
  n9 = FunctionInput(Preview=StaticBool() {value=true}) {
   bUsePreviewValueAsDefault=true
   description="Default = True"
   inputName=UseTangentNormals
   inputType=FunctionInput_StaticBool
   sortPriority=1
  }
  n10 = MaterialFunctionCall(Scale=NamedRerouteUsage()) {materialFunction=&MF_SimpleWorldUV}
  n11 = TextureSample(UVs=@n10.Result) {
   samplerSource=SSM_Wrap_WorldGroupSettings
   samplerType=SAMPLERTYPE_Masks
   texture=/Game/Athena/Apollo/Environments/Landscape/Textures/Roads/Road_Cracks_M
  }
  n12 = NamedRerouteDeclaration(None=ComponentMask(None=@n11.RGB) {g=true}) {name=CrackHeight, nodeColor=<rgba 0.5469 0 1 1>}
  n13 = LinearInterpolate(Alpha=@n3.B) {constA=0.5, constB=0}
  n14 = SetMaterialAttributes!(
   MaterialAttributes
   BaseColor=Multiply(
    A=NamedRerouteUsage()
    B=TextureSample(UVs=@n1.Result) {samplerSource=SSM_Wrap_WorldGroupSettings, texture=/Game/Athena/Apollo/Environments/Landscape/Textures/Roads/T_AsphaltDetail1_D}.RGB
   )
   Specular=@n3.R
   Roughness=@n3.R
   Normal=MaterialFunctionCall!(Normals=LinearInterpolate(A=@n4, B=@n5, Alpha=@n8.Lerp Alpha No Contrast), TransformToTangentSpace=@n9, Normal Strength) {materialFunction=/Game/Athena/Apollo/Environments/Landscape/Materials/Biome/CommonMaterials/MF_HQ_WorldAlignedNormal}.Result
   OpacityMask=NamedRerouteUsage()
   Displacement=LinearInterpolate(
    B=LinearInterpolate(A=@n12, B=@n13, Alpha=@n8.Lerp Alpha No Contrast)
    Alpha=ScalarParameter() {defaultValue=1, group=Displacement, parameterName=Asphalt_DisplacementAmt}
   ) {constA=0.5}
  ) {attributeSetTypes [
    #1
    #2
    #3
    #4
    #5
    #6
   ]}
  FunctionOutput(None=@n14)
  FunctionOutput(None=@n8.Lerp Alpha No Contrast) {outputName=CrackHeightWithVariationMask, sortPriority=500}
  NamedRerouteDeclaration(None=Constant() {r=2000}) {name=CrackScale, nodeColor=<rgba 0.5469 0 1 1>}
  n15 = NamedRerouteUsage()
  n16 = ComponentMask(None=@n11.RGB) {r=true}
  NamedRerouteDeclaration(None=OneMinus(None=Multiply(A=@n15, B=@n16))) {name=CrackDiffuse, nodeColor=<rgba 0.5469 0 1 1>}
  NamedRerouteDeclaration(None=Constant() {r=350}) {name=AsphaltScale, nodeColor=<rgba 0.5469 0 1 1>}
  NamedRerouteDeclaration(None=Constant() {r=0.5905}) {name=AsphaltVariation, nodeColor=<rgba 0.5469 0 1 1>}
  n17 = OneMinus(None=@n3.B)
  n18 = Multiply(A=@n17, B=Constant() {r=1.35}) {constB=2}
  NamedRerouteDeclaration(None=MaterialFunctionCall(Transition Phase=NamedRerouteUsage(), Height Texture=Saturate(None=@n18)) {materialFunction=&HeightLerp}.Lerp Alpha No Contrast) {name=Asphalt_Height, nodeColor=<rgba 0.5469 0 1 1>}
  n19 = TextureSample(UVs=@n1.Result) {
   samplerSource=SSM_Wrap_WorldGroupSettings
   samplerType=SAMPLERTYPE_Normal
   texture=/Game/Athena/Apollo/Environments/Landscape/Textures/Roads/T_AsphaltDetail1_N
  }
  NamedRerouteDeclaration(None=@n19.RGB) {name=AsphaltNormal, nodeColor=<rgba 0.5469 0 1 1>}
  n20 = TextureSample(UVs=@n10.Result) {
   samplerSource=SSM_Wrap_WorldGroupSettings
   samplerType=SAMPLERTYPE_Normal
   texture=/Game/Athena/Apollo/Environments/Landscape/Textures/Roads/Road_Cracks_N
  }
  NamedRerouteDeclaration(None=@n20.RGB) {name=CrackNormal, nodeColor=<rgba 0.5469 0 1 1>}
  Reroute(None=@n12)
  Reroute(None=MaterialFunctionCall(Transition Phase=NamedRerouteUsage(), Height Texture=@n17) {materialFunction=&HeightLerp}.Lerp Alpha No Contrast)
  ScalarParameter() {defaultValue=0.5905, group=Asphalt, parameterName=AsphaltVariation}
  ScalarParameter() {defaultValue=350, group=Asphalt, parameterName=AsphaltMainScale}
  ScalarParameter() {defaultValue=2000, group=Asphalt, parameterName=SizeCracks}
  TextureCoordinate() {uTiling=6, vTiling=6}
  TextureCoordinate()
 ]}
