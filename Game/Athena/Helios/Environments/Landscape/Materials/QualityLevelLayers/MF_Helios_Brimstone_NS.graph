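# Material function that builds the Helios Brimstone landscape layer from one packed mobile mask texture, outputting base color, specular, roughness, normal and opacity mask attributes.
# The texture is sampled by world XY through a periodic-world transform, with RGB feeding normal, blue a texture mask and alpha height.
# A scaled, saturated mask becomes dirt; named reroutes feed parameterized roughness and a three-color blend, darkened by an absolute offset-from-0.5 term.
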MaterialFunction /Game/Athena/Helios/Environments/Landscape/Materials/QualityLevelLayers/MF_Helios_Brimstone_NS
graph {nodes [
  n1 = SetMaterialAttributes!(
   MaterialAttributes
   BaseColor=NamedRerouteUsage()
   Specular=NamedRerouteUsage()
   Roughness=NamedRerouteUsage()
   Normal=NamedRerouteUsage()
   OpacityMask=NamedRerouteUsage()
  ) {
   desc=HighEnd Material
   attributeSetTypes [
    #1
    #2
    #3
    #4
    #5
   ]
  }
  FunctionOutput(None=@n1)
  n2 = Constant() {r=750}
  n3 = TransformPosition!(None=WorldPosition() {worldPositionShaderOffset=WPT_CameraRelative}.XYZ, Periodic World Tile Size=@n2) {transformSourceType=TRANSFORMPOSSOURCE_TranslatedWorld, transformType=TRANSFORMPOSSOURCE_PeriodicWorld}
  n4 = TextureSample(UVs=ComponentMask(None=Divide(A=@n3, B=@n2)) {g=true, r=true}) {samplerType=SAMPLERTYPE_Masks, texture=/Game/Environments/Helios/Terrain/TexturesMobile/T_Helios_Mobile_Brimstone}
  NamedRerouteDeclaration(None=@n4.A) {name=Height, nodeColor=<rgba 1 0 0.5156 1>}
  n5 = Multiply(A=NamedRerouteUsage()) {constB=9}
  NamedRerouteDeclaration(None=Saturate(None=Subtract(A=@n5) {constB=2})) {name=Dirt Mask, nodeColor=<rgba 0.2656 1 0 1>}
  n6 = NamedRerouteUsage()
  NamedRerouteDeclaration(None=LinearInterpolate(
    A=ScalarParameter() {defaultValue=0.95, group=Mobile Brimstone, parameterName=Mobile Brimstone Roughness Min}
    B=ScalarParameter() {defaultValue=0.65, group=Mobile Brimstone, parameterName=Mobile Brimstone Roughness Max}
    Alpha=Multiply(A=@n6, B=@n6)
   )) {name=Roughness, nodeColor=<rgba 0 1 0.9531 1>}
  NamedRerouteDeclaration(None=ScalarParameter() {defaultValue=0.65, group=Mobile Brimstone, parameterName=Mobile Brimstone Specular}) {name=Specular, nodeColor=<rgba 0.0078 0 1 1>}
  NamedRerouteDeclaration(None=@n4.RGB) {name=Normal, nodeColor=<rgba 0.9219 1 0 1>}
  n7 = VectorParameter() {group=Mobile Brimstone, parameterName=Mobile Brimstone Color 1, defaultValue=<rgba 0.0471 0.0392 0.0392 1>}
  n8 = VectorParameter() {group=Mobile Brimstone, parameterName=Mobile Brimstone Color 2, defaultValue=<rgba 0.051 0.0588 0.0627 1>}
  n9 = LinearInterpolate(A=@n7.RGB, B=@n8.RGB, Alpha=NamedRerouteUsage())
  n10 = VectorParameter() {group=Mobile Brimstone, parameterName=Mobile Brimstone Color 3, defaultValue=<rgba 0.1686 0.102 0.051 1>}
  n11 = NamedRerouteUsage()
  n12 = Abs(None=Add(A=NamedRerouteUsage()) {constB=-0.5})
  n13 = Multiply(A=Add(A=ComponentMask(None=@n12) {r=true}, B=ComponentMask(None=@n12) {r=true})) {constB=0.5}
  NamedRerouteDeclaration(None=Multiply(A=LinearInterpolate(A=@n9, B=@n10.RGB, Alpha=@n11), B=OneMinus(None=@n13))) {name=Base Color, nodeColor=<rgba 0.8984 1 0 1>}
  NamedRerouteDeclaration(None=@n4.B) {name=Texture Mask, nodeColor=<rgba 0.9688 0 1 1>}
 ]}
