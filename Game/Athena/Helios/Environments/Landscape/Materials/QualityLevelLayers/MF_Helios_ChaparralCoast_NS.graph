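# Landscape layer material function for Helios chaparral coast ground cover, built on mobile-specific textures and parameters.
# One world-aligned packed texture supplies normal, color mask and height.
# Masked color parameters form a base color with bright flower tinting, while scalar parameters shape specular and roughness.
# It outputs material attributes with and without flowers, plus mobile noise and layer masks, the noise built from multiplied, saturated channels.
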MaterialFunction /Game/Athena/Helios/Environments/Landscape/Materials/QualityLevelLayers/MF_Helios_ChaparralCoast_NS
graph {nodes [
  n1 = NamedRerouteUsage()
  n2 = ComponentMask(None=@n1) {b=true}
  n3 = ComponentMask(None=@n1) {g=true}
  n4 = Saturate(None=Multiply(A=Multiply(A=@n2, B=@n3)) {constB=3})
  n5 = NamedRerouteUsage()
  FunctionOutput(None=Add(A=Multiply(A=@n4, B=@n5)) {constB=0.1}) {outputName=MobileLayerMask, sortPriority=1}
  FunctionOutput(None=@n4) {outputName=MobileNoiseMask, sortPriority=50}
  n6 = SetMaterialAttributes!(
   MaterialAttributes
   BaseColor=NamedRerouteUsage()
   Specular=NamedRerouteUsage()
   Roughness=NamedRerouteUsage()
   Normal=NamedRerouteUsage()
   OpacityMask=NamedRerouteUsage()
  ) {
   desc=HighEnd Material
   attributeSetTypes [
    #1
    #2
    #3
    #4
    #5
   ]
  }
  FunctionOutput(None=@n6)
  n7 = SetMaterialAttributes!(
   MaterialAttributes
   BaseColor=NamedRerouteUsage()
   Specular=NamedRerouteUsage()
   Roughness=NamedRerouteUsage()
   Normal=NamedRerouteUsage()
   OpacityMask=NamedRerouteUsage()
   Opacity=NamedRerouteUsage()
  ) {
   desc=HighEnd Material
   attributeSetTypes [
    #1
    #2
    #3
    #4
    #5
    #6
   ]
  }
  FunctionOutput(None=@n7) {outputName=ResultNoFlowers, sortPriority=1}
  n8 = WorldPosition()
  n9 = TextureSample(UVs=Divide(A=TruncateLWC!(None=@n8.XY), B=ScalarParameter() {defaultValue=900, group=Mobile Chaparral, parameterName=Mobile Chaparral Scale}) {constB=1200}) {samplerType=SAMPLERTYPE_Masks, texture=/Game/Environments/Helios/Terrain/TexturesMobile/T_Helios_Mobile_ChaparralCoast}
  NamedRerouteDeclaration(None=@n9.RGB) {name=Normal, nodeColor=<rgba 0.3079 0.0373 1 1>}
  NamedRerouteDeclaration(None=FunctionInput() {bUsePreviewValueAsDefault=true, inputName=Layer Weight, inputType=FunctionInput_Scalar}) {name=Layer Weight, nodeColor=<rgba 0.5234 1 0 1>}
  NamedRerouteDeclaration(None=@n9.B) {name=Color Mask, nodeColor=<rgba 0 0.2031 1 1>}
  NamedRerouteDeclaration(None=@n9.A) {name=Height, nodeColor=<rgba 0.8984 1 0 1>}
  n10 = VectorParameter() {group=Mobile Chaparral, parameterName=Mobile Chaparral Color 1, defaultValue=<rgba 0.0158 0.0781 0.0481 1>}
  n11 = VectorParameter() {group=Mobile Chaparral, parameterName=Mobile Chaparral Color 2, defaultValue=<rgba 0.2541 0.2969 0.0959 1>}
  n12 = OneMinus(None=ComponentMask(None=NamedRerouteUsage()) {b=true})
  n13 = Saturate(None=Subtract(A=Multiply(A=@n12) {constB=7}) {constB=4.5})
  n14 = VectorParameter() {group=Mobile Chaparral, parameterName=Mobile Chaparral Color3, defaultValue=<rgba 0 0.1108 0.1667 1>}
  n15 = Multiply(A=LinearInterpolate(A=@n10.RGB, B=@n11.RGB, Alpha=NamedRerouteUsage()), B=Add(A=Multiply(A=@n13, B=@n14.RGB)))
  n16 = NamedRerouteUsage()
  n17 = Subtract(A=Multiply(A=ComponentMask(None=@n16) {b=true}) {constB=7}) {constB=3}
  n18 = Subtract(A=Multiply(A=NamedRerouteUsage()) {constB=5}) {constB=2}
  n19 = Multiply(
   A=Multiply(A=Saturate(None=@n17), B=Saturate(None=@n18))
   B=VectorParameter() {group=Mobile Chaparral, parameterName=Mobile Chaparral Flowers, defaultValue=<rgba 1.7 0.9605 0.142 1>}.RGB
  )
  NamedRerouteDeclaration(None=Multiply(A=NamedRerouteDeclaration(None=@n15) {name=BaseColorNoFlowers, nodeColor=<rgba 0 1 0.8828 1>}, B=Add(A=@n19))) {name=BaseColor, nodeColor=<rgba 0.5 0 1 1>}
  NamedRerouteDeclaration(None=FunctionInput() {bUsePreviewValueAsDefault=true, inputName=VariationMasks Mobile, inputType=FunctionInput_Vector4, sortPriority=1}) {name=VariationMasks Mobile, nodeColor=<rgba 1 0.7031 0 1>}
  n20 = NamedRerouteUsage()
  n21 = ScalarParameter() {defaultValue=0.5, group=Mobile Chaparral, parameterName=Mobile Chaparral Specular}
  NamedRerouteDeclaration(None=Saturate(None=Multiply(A=@n20, B=@n21))) {name=Specular, nodeColor=<rgba 0.5 0 1 1>}
  n22 = Multiply(A=NamedRerouteUsage(), B=ScalarParameter() {defaultValue=0.5, group=Mobile Chaparral, parameterName=Mobile Chaparral Roughness})
  NamedRerouteDeclaration(None=Saturate(None=OneMinus(None=@n22))) {name=Roughness, nodeColor=<rgba 0.1484 0 1 1>}
 ]}
