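# Default quality-level landscape layer function for the Helios DesertOasis ground, outputting material attributes (also as StaticMeshMaterial) and a FlowerMask.
# World-projected chaparral coast textures supply tinted color, world-aligned normals, height, and packed specular, flower mask and roughness.
# A separate flower tint blends over the base color through a product of masks.
# Two scales of world-space noise, shaped by Beer's law and power, form a layer breakup mask.
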MaterialFunction /Game/Athena/Helios/Environments/Landscape/Materials/QualityLevelLayers/MF_Helios_DesertOasis_Default
graph {nodes [
  n1 = MaterialFunctionCall!(Scale=ScalarParameter() {defaultValue=900, group=DesertOasis, parameterName=DesertOasis_WorldScale}) {materialFunction=/Game/Athena/Apollo/Environments/Landscape/Materials/Biome/CommonMaterials/MF_SimpleWorldUV}
  n2 = TextureSample(
   UVs=@n1.Result
   Tex=TextureObject() {samplerType=SAMPLERTYPE_Masks, texture=/Game/Environments/Helios/Terrain/Textures/T_Helios_Chap_CoastGround_01_S}
  ) {
   samplerSource=SSM_Wrap_WorldGroupSettings
   samplerType=SAMPLERTYPE_Masks
   texture=&T_Helios_Chap_CoastGround_01_S
  }
  n3 = MaterialFunctionCall(Float3=@n2.RGB) {materialFunction=&BreakOutFloat3Components}
  n4 = NamedRerouteDeclaration(None=@n3.B) {name=Grass_ChaparralCoast_Roughness, nodeColor=<rgba 0.5 0 1 1>}
  n5 = ScalarParameter() {defaultValue=0.5, group=DesertOasis, parameterName=DesertOasis_RoughAdd}
  n6 = SetMaterialAttributes!(
   MaterialAttributes
   BaseColor=NamedRerouteUsage()
   Specular=LinearInterpolate(
    B=NamedRerouteDeclaration(None=@n3.R) {name=Grass_ChaparralCoast_Spec, nodeColor=<rgba 0.5 0 1 1>}
    Alpha=ScalarParameter() {defaultValue=1, group=DesertOasis, parameterName=DesertOasis_SpecAmt}
   ) {constA=0.5}
   Roughness=Saturate(None=Add(A=@n4, B=@n5))
   Normal=NamedRerouteUsage()
   OpacityMask=NamedRerouteUsage()
   Displacement=NamedRerouteUsage()
  ) {
   desc=HighEnd Material
   attributeSetTypes [
    #1
    #2
    #3
    #4
    #5
    #6
   ]
  }
  FunctionOutput(None=@n6) {sortPriority=-100}
  FunctionOutput(None=@n6) {outputName=StaticMeshMaterial, sortPriority=-100}
  n7 = NamedRerouteUsage()
  n8 = ScalarParameter() {defaultValue=0.1, group=DesertOasis, parameterName=DesertOasis_Fmask_Add_Grass}
  FunctionOutput(None=Saturate(None=Add(A=@n7, B=@n8))) {outputName=FlowerMask, sortPriority=500}
  n9 = NamedRerouteUsage()
  NamedRerouteDeclaration(None=LinearInterpolate(B=@n9, Alpha=ScalarParameter() {defaultValue=1, group=Displacement, parameterName=DesertOasis_DisplacementAmt}) {constA=0.5}) {name=Displacement, nodeColor=<rgba 0.5 0 1 1>}
  NamedRerouteDeclaration(None=@n3.G) {name=FlowerMask, nodeColor=<rgba 0.5 0 1 1>}
  n10 = TextureSample(UVs=@n1.Result, Tex=TextureObject() {texture=/Game/Environments/Helios/Terrain/Textures/T_Helios_Chap_CoastGround_01_D}) {samplerSource=SSM_Wrap_WorldGroupSettings, texture=&T_Helios_Chap_CoastGround_01_D}
  NamedRerouteDeclaration(None=@n10.A) {name=DesertOasis_Height, nodeColor=<rgba 0.5 0 1 1>}
  n11 = VectorParameter() {group=DesertOasis, parameterName=DesertOasis_ColorMod, defaultValue=<rgba 0.2907 0.5 0.05 0.7>}
  n12 = Multiply(A=@n11.RGB, B=Power(Base=@n10.RGB, Exp=@n11.A))
  n13 = Multiply(A=VectorParameter() {group=DesertOasis, parameterName=DesertOasis_Flower_ColorMod, defaultValue=<rgba 1.0588 0.275 1.25 1>}.RGB, B=@n10.G)
  n14 = NamedRerouteUsage()
  NamedRerouteDeclaration(None=LinearInterpolate(A=Saturate(None=@n12), B=Saturate(None=@n13), Alpha=Multiply(A=@n14, B=@n7))) {name=DesertOasis_Diffuse, nodeColor=<rgba 0.5 0 1 1>}
  n15 = TextureSample(
   UVs=@n1.Result
   Tex=TextureObject() {samplerType=SAMPLERTYPE_Normal, texture=/Game/Environments/Helios/Terrain/Textures/T_Helios_Chap_CoastGround_01_N}
  ) {
   samplerSource=SSM_Wrap_WorldGroupSettings
   samplerType=SAMPLERTYPE_Normal
   texture=&T_Helios_Chap_CoastGround_01_N
  }
  n16 = FunctionInput(Preview=StaticBool()) {
   bUsePreviewValueAsDefault=true
   description="Default = False"
   inputName=UseTangentNormals
   inputType=FunctionInput_StaticBool
   sortPriority=1
  }
  NamedRerouteDeclaration(None=MaterialFunctionCall!(Normals=@n15.RGB, TransformToTangentSpace=@n16, Normal Strength=Constant() {r=1}) {materialFunction=/Game/Athena/Apollo/Environments/Landscape/Materials/Biome/CommonMaterials/MF_HQ_WorldAlignedNormal}.Result) {name=DesertOasis_Normal, nodeColor=<rgba 0.5 0 1 1>}
  n17 = TextureObject() {texture=/Game/Environments/Helios/Terrain/Textures/T_Helios_Landscape_Noise}
  n18 = TextureSample(
   UVs=MaterialFunctionCall(Scale=ScalarParameter() {defaultValue=1400, group=DesertOasis, parameterName=DesertOasis_NoiseScale01}) {materialFunction=&MF_SimpleWorldUV}.Result
   Tex=@n17
  ) {samplerSource=SSM_Wrap_WorldGroupSettings, texture=/Game/Global/Textures/Default/Noise/T_Tiling_Noise_01}
  n19 = ScalarParameter() {defaultValue=500, group=DesertOasis, parameterName=DesertOasis_NoiseScale02}
  n20 = Multiply(
   A=@n18.B
   B=TextureSample(
    UVs=MaterialFunctionCall(Scale=@n19) {materialFunction=&MF_SimpleWorldUV}.Result
    Tex=@n17
   ) {samplerSource=SSM_Wrap_WorldGroupSettings, texture=&T_Tiling_Noise_01}.B
  )
  n21 = Multiply(
   A=OneMinus(None=@n18.B)
   B=MaterialFunctionCall!(
    Thickness=SquareRoot(None=@n20)
    Depth Scale=ScalarParameter() {defaultValue=1.5, group=DesertOasis, parameterName=DesertOasis_Noise_BeersLaw}
   ) {materialFunction=/Engine/Functions/Engine_MaterialFunctions01/Volumetrics/BeersLaw}.Result
  )
  n22 = Multiply(
   A=Power(Base=@n21, Exp=ScalarParameter() {defaultValue=1.5, group=DesertOasis, parameterName=DesertOasis_Noise_Power})
   B=ScalarParameter() {defaultValue=6, group=DesertOasis, parameterName=DesertOasis_Fmask_Mod}
  )
  n23 = ScalarParameter() {defaultValue=-1.75, group=DesertOasis, parameterName=DesertOasis_Fmask_Add}
  NamedRerouteDeclaration(None=Saturate(None=Add(A=@n22, B=@n23))) {name=LandscapeLayerMaskNoise, nodeColor=<rgba 1 0 0.063 1>}
  OneMinus(None=@n21)
  n24 = ScalarParameter() {defaultValue=1, group=Displacement, parameterName=DesertOasis_DisplacementPower}
  Saturate(None=Multiply(A=Power(Base=@n9, Exp=@n24), B=ScalarParameter() {defaultValue=1, group=Displacement, parameterName=DesertOasis_DisplacementMod}))
 ]}
