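# Material function that builds the mobile Desert Oasis grass layer for Helios landscape materials, adding flower color to tinted grass.
# The blue variation-mask channel is scaled, inverted, then contrast-adjusted by multiply/offset parameters into the Flower Mask output.
# The same mask, multiplied by the grass OpacityMask before that adjustment, blends tinted grass BaseColor toward the flower color.
# Only BaseColor is overridden in the returned material attributes.
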MaterialFunction /Game/Athena/Helios/Environments/Landscape/Materials/QualityLevelLayers/MF_Helios_DesertOasis_NS
graph {nodes [
  n1 = NamedRerouteUsage()
  n2 = OneMinus(None=Multiply(A=ComponentMask(None=@n1) {b=true}) {constB=0.4})
  n3 = ScalarParameter() {defaultValue=20, group=Mobile DesertOasis, parameterName=Mobile DesertOasisMult}
  n4 = Multiply(A=@n2, B=@n3) {constB=5}
  n5 = ScalarParameter() {defaultValue=-4.2, group=Mobile DesertOasis, parameterName=Mobile DesertOasisSub}
  FunctionOutput(None=Saturate(None=Add(A=@n4, B=@n5) {constB=-1})) {outputName=Flower Mask, sortPriority=1}
  n6 = GetMaterialAttributes(None=NamedRerouteUsage()) {attributeGetTypes [#1, #2]}
  n7 = Multiply(
   A=@n6.BaseColor
   B=VectorParameter() {group=Mobile DesertOasis, parameterName=Mobile DesertOasisColor, defaultValue=<rgba 0.5625 1 0.5679 1>}.RGB
  )
  n8 = VectorParameter() {group=Mobile DesertOasis, parameterName=Mobile DesertOasisFlowers, defaultValue=<rgba 0.2812 0.0703 0.1662 1>}
  n9 = Multiply(A=@n6.OpacityMask, B=@n2) {constB=3}
  n10 = Saturate(None=Add(A=Multiply(A=@n9, B=@n3) {constB=5}, B=@n5) {constB=-1})
  FunctionOutput(None=SetMaterialAttributes!(MaterialAttributes=@n6.MaterialAttributes, BaseColor=LinearInterpolate(A=@n7, B=@n8.RGB, Alpha=@n10)) {attributeSetTypes [#1]})
  NamedRerouteDeclaration(None=FunctionInput() {bUsePreviewValueAsDefault=true, inputName=VariationMasks Mobile, inputType=FunctionInput_Vector4, sortPriority=1}) {name=VariationMasks Mobile, nodeColor=<rgba 1 0.7031 0 1>}
  NamedRerouteDeclaration(None=FunctionInput() {bUsePreviewValueAsDefault=true, inputName=MobileGrass, inputType=FunctionInput_MaterialAttributes}) {name=MobileGrass, nodeColor=<rgba 1 0.1504 0.851 1>}
 ]}
