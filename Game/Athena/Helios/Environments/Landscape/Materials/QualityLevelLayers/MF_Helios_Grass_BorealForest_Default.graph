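# High-end landscape material layer function for boreal forest grass in the Helios environment, outputting material attributes and a layer weight.
# World-space UVs sample forest floor diffuse, packed mask and normal textures, with named reroutes carrying color, specular, roughness, height and normal into the attributes, including opacity mask and displacement.
# Inverted landscape noise from MF_ScotchBroom_Colors masks the incoming LayerWeight.
# The flower color blend and displacement power chain feed no output.
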MaterialFunction /Game/Athena/Helios/Environments/Landscape/Materials/QualityLevelLayers/MF_Helios_Grass_BorealForest_Default
graph {nodes [
  n1 = NamedRerouteUsage()
  n2 = NamedRerouteUsage()
  n3 = SetMaterialAttributes!(
   MaterialAttributes
   BaseColor=@n2
   Specular=NamedRerouteUsage()
   Roughness=NamedRerouteUsage()
   Normal=NamedRerouteUsage()
   OpacityMask=@n1
   Displacement=LinearInterpolate(B=@n1, Alpha=ScalarParameter() {defaultValue=1, group=Grass_BorealForest, parameterName=Grass_BorealForest_DisplacementAmt}) {constA=0.5}
  ) {
   desc=HighEnd Material
   attributeSetTypes [
    #1
    #2
    #3
    #4
    #5
    #6
   ]
  }
  FunctionOutput(None=@n3) {sortPriority=-100}
  n4 = Constant()
  n5 = MaterialFunctionCall() {materialFunction=/Game/Athena/Helios/Environments/Landscape/Materials/MaterialFunctions_Utilities/MF_ScotchBroom_Colors}
  n6 = OneMinus(None=@n5.LandscapeLayerMaskNoise)
  FunctionOutput(None=Multiply(
    A=FunctionInput(Preview=@n4) {bUsePreviewValueAsDefault=true, inputName=LayerWeight, inputType=FunctionInput_Scalar, sortPriority=-500}
    B=@n6
   )) {outputName=HighEnd_LayerWeight AND BorealForest_GrassMask, sortPriority=-1}
  n7 = NamedRerouteUsage()
  n8 = Add(
   A=Multiply(A=@n6, B=ScalarParameter() {defaultValue=6, group=Grass_BorealForest, parameterName=Grass_BorealForest_Fmask_Mod})
   B=ScalarParameter() {defaultValue=-3, group=Grass_BorealForest, parameterName=Grass_BorealForest_Fmask_Add}
  )
  LinearInterpolate(A=@n2, B=@n5.Flower1_Color, Alpha=Multiply(A=Multiply(A=@n5.FlowerWorldMask1, B=@n7), B=Saturate(None=@n8)))
  n9 = MaterialFunctionCall!(Scale=ScalarParameter() {defaultValue=900, group=Grass_BorealForest, parameterName=Grass_BorealForest_Scale}) {materialFunction=/Game/Athena/Apollo/Environments/Landscape/Materials/Biome/CommonMaterials/MF_SimpleWorldUV}
  n10 = TextureSample(UVs=@n9.Result) {
   samplerSource=SSM_Wrap_WorldGroupSettings
   samplerType=SAMPLERTYPE_Masks
   texture=/Game/Environments/Helios/Terrain/Textures/T_Helios_Bor_ForestFloor_01_S
  }
  n11 = MaterialFunctionCall(Float3=@n10.RGB) {materialFunction=&BreakOutFloat3Components}
  NamedRerouteDeclaration(None=@n11.G) {name=FlowerMask, nodeColor=<rgba 0.5 0 1 1>}
  n12 = TextureSample(UVs=@n9.Result) {
   samplerSource=SSM_Wrap_WorldGroupSettings
   samplerType=SAMPLERTYPE_Normal
   texture=/Game/Environments/Helios/Terrain/Textures/T_Helios_Bor_ForestFloor_01_N
  }
  n13 = FunctionInput(Preview=StaticBool()) {
   bUsePreviewValueAsDefault=true
   description="Default = False"
   inputName=UseTangentNormals
   inputType=FunctionInput_StaticBool
   sortPriority=1
  }
  NamedRerouteDeclaration(None=MaterialFunctionCall!(
    Normals=@n12.RGB
    TransformToTangentSpace=@n13
    Normal Strength=ScalarParameter() {defaultValue=1, group=Grass_BorealForest, parameterName=Grass_BorealForest_NormalAmt}
   ) {materialFunction=/Game/Athena/Apollo/Environments/Landscape/Materials/Biome/CommonMaterials/MF_HQ_WorldAlignedNormal}.Result) {name=Normal, nodeColor=<rgba 0.5 0 1 1>}
  n14 = TextureSample(UVs=@n9.Result) {samplerSource=SSM_Wrap_WorldGroupSettings, texture=/Game/Environments/Helios/Terrain/Textures/T_Helios_Bor_ForestFloor_01_D}
  NamedRerouteDeclaration(None=@n14.A) {name=Grass_BorealForest_Height, nodeColor=<rgba 0 0.5 1 1>}
  NamedRerouteDeclaration(None=@n14.RGB) {name=Grass_BorealForest_Diffuse, nodeColor=<rgba 0 0.5 1 1>}
  NamedRerouteDeclaration(None=@n11.R) {name=Specular, nodeColor=<rgba 0.5 0 1 1>}
  NamedRerouteDeclaration(None=@n11.B) {name=Roughness, nodeColor=<rgba 0.5 0 1 1>}
  n15 = ScalarParameter() {defaultValue=2, group=Grass_BorealForest, parameterName=Grass_BorealForest_DisplacementPower}
  Saturate(None=Multiply(
    A=Power(Base=@n1, Exp=@n15)
    B=ScalarParameter() {defaultValue=2, group=Grass_BorealForest, parameterName=Grass_BorealForest_DisplacementMod}
   ))
 ]}
