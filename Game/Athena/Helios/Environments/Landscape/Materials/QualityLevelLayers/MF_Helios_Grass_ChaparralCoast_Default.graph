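# Landscape layer material function for Helios Chaparral Coast grass, outputting high-end and static-mesh material attributes, a weighted layer mask and a noise mask.
# The landscape path samples diffuse, normal and packed specular/flower/roughness textures through hex-grid tiling functions and adds height-based displacement.
# The static-mesh path samples the same textures with simple world UVs and world-aligned normals.
# MF_ScotchBroom_Colors supplies flower color, world mask and layer noise for blending over tinted grass.
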MaterialFunction /Game/Athena/Helios/Environments/Landscape/Materials/QualityLevelLayers/MF_Helios_Grass_ChaparralCoast_Default
graph {nodes [
  n1 = VectorParameter() {group=Grass_ChaparralCoast, parameterName=ChaparralCoastGrassTextureTint, defaultValue=<rgba 0.7953 0.95 0.7125 1>}
  n2 = NamedRerouteUsage()
  n3 = Multiply(A=NamedRerouteUsage(), B=NamedRerouteUsage())
  n4 = NamedRerouteUsage()
  n5 = ScalarParameter() {defaultValue=6, group=Grass_ChaparralCoast, parameterName=Grass_ChaparralCoast_Fmask_Mod}
  n6 = Saturate(None=Add(A=Multiply(A=@n4, B=@n5), B=ScalarParameter() {defaultValue=-3, group=Grass_ChaparralCoast, parameterName=Grass_ChaparralCoast_Fmask_Add}))
  n7 = SetMaterialAttributes!(
   MaterialAttributes
   BaseColor=LinearInterpolate(A=Multiply(A=@n1.RGB, B=@n2), B=NamedRerouteUsage(), Alpha=Multiply(A=@n3, B=@n6))
   Specular=NamedRerouteUsage()
   Roughness=NamedRerouteUsage()
   Normal=NamedRerouteUsage()
   OpacityMask=NamedRerouteUsage()
   Displacement=NamedRerouteUsage()
  ) {
   desc=HighEnd Material
   attributeSetTypes [
    #1
    #2
    #3
    #4
    #5
    #6
   ]
  }
  FunctionOutput(None=@n7) {sortPriority=-100}
  n8 = Constant()
  n9 = NamedRerouteUsage()
  FunctionOutput(None=Multiply(
    A=FunctionInput(Preview=@n8) {bUsePreviewValueAsDefault=true, inputName=LayerWeight, inputType=FunctionInput_Scalar, sortPriority=-500}
    B=@n9
   )) {outputName=HighEnd_LayerWeight AND ChaparralCoast_GrassMask, sortPriority=-1}
  n10 = SetMaterialAttributes!(
   MaterialAttributes
   BaseColor=NamedRerouteUsage()
   Specular=NamedRerouteUsage()
   Roughness=NamedRerouteUsage()
   Normal=NamedRerouteUsage()
   OpacityMask=NamedRerouteUsage()
  ) {
   desc=Static Mesh Material
   attributeSetTypes [
    #1
    #2
    #3
    #4
    #5
   ]
  }
  FunctionOutput(None=@n10) {outputName=StaticMeshMaterial, sortPriority=-100}
  FunctionOutput(None=@n9) {outputName=NoiseMask, sortPriority=500}
  n11 = NamedRerouteUsage()
  NamedRerouteDeclaration(None=LinearInterpolate(B=@n11, Alpha=ScalarParameter() {defaultValue=1, group=Displacement, parameterName=Grass_ChaparralCoast_DisplacementAmt}) {constA=0.5}) {name=Displacement, nodeColor=<rgba 0.5 0 1 1>}
  n12 = FunctionInput(Preview=TextureObject() {samplerType=SAMPLERTYPE_Masks, texture=/Game/Environments/Helios/Terrain/Textures/T_Helios_Chap_CoastGround_01_S}) {bUsePreviewValueAsDefault=true, inputName=Chap_S, inputType=FunctionInput_Texture2D, sortPriority=-399}
  n13 = Multiply(A=TextureCoordinate(), B=ScalarParameter() {defaultValue=0.19, group=Grass_ChaparralCoast, parameterName=GR_CoordMod})
  n14 = ScalarParameter() {defaultValue=6, group=Grass_ChaparralCoast, parameterName=GR_HexGridSharp}
  n15 = ScalarParameter() {defaultValue=0.75, group=Grass_ChaparralCoast, parameterName=GR_HexGridSize}
  n16 = ScalarParameter() {defaultValue=1, group=Grass_ChaparralCoast, parameterName=GR_MinScale}
  n17 = ScalarParameter() {defaultValue=1.5, group=Grass_ChaparralCoast, parameterName=GR_MaxScale}
  n18 = MaterialFunctionCall(Float3=ComponentMask(None=MaterialFunctionCall!(
     Texture=@n12
     LandscapeCoordinates=@n13
     LowEndTiling
     HexGridGradientSharpness=@n14
     HexGridSize=@n15
     HexGridUVInput
     MinScale=@n16
     MaxScale=@n17
     MinOffset
     MaxOffset
     "MaxRotation(Degrees)"
     "MinRotation(Degrees)"
     forceLowQuality
    ) {materialFunction=/Game/Athena/Asteria/Environments/Landscape/Materials/MaterialFunctions/UtilityFunctions/Landscape_TextureSample}.Result) {b=true, g=true, r=true}) {materialFunction=&BreakOutFloat3Components}
  NamedRerouteDeclaration(None=@n18.G) {name=FlowerMask, nodeColor=<rgba 0.5 0 1 1>}
  n19 = MaterialFunctionCall!(Scale=Constant() {r=900}) {materialFunction=/Game/Athena/Apollo/Environments/Landscape/Materials/Biome/CommonMaterials/MF_SimpleWorldUV}
  n20 = TextureSample(UVs=@n19.Result, Tex=@n12) {
   samplerSource=SSM_Wrap_WorldGroupSettings
   samplerType=SAMPLERTYPE_Masks
   texture=&T_Helios_Chap_CoastGround_01_S
  }
  n21 = MaterialFunctionCall(Float3=@n20.RGB) {materialFunction=&BreakOutFloat3Components}
  NamedRerouteDeclaration(None=@n21.G) {name=FlowerMask_SM, nodeColor=<rgba 0.5 0 1 1>}
  n22 = FunctionInput(Preview=TextureObject() {samplerType=SAMPLERTYPE_Normal, texture=/Game/Environments/Helios/Terrain/Textures/T_Helios_Chap_CoastGround_01_N}) {bUsePreviewValueAsDefault=true, inputName=Chap_N, inputType=FunctionInput_Texture2D, sortPriority=-398}
  n23 = TextureSample(UVs=@n19.Result, Tex=@n22) {
   samplerSource=SSM_Wrap_WorldGroupSettings
   samplerType=SAMPLERTYPE_Normal
   texture=&T_Helios_Chap_CoastGround_01_N
  }
  n24 = FunctionInput(Preview=StaticBool()) {
   bUsePreviewValueAsDefault=true
   description="Default = False"
   inputName=UseTangentNormals
   inputType=FunctionInput_StaticBool
   sortPriority=-499
  }
  NamedRerouteDeclaration(None=MaterialFunctionCall!(
    Normals=@n23.RGB
    TransformToTangentSpace=@n24
    Normal Strength=ScalarParameter() {defaultValue=0.5, group=Grass_ChaparralCoast, parameterName=Grass_ChaparralCoast_StaticMeshNormalStrength}
   ) {materialFunction=/Game/Athena/Apollo/Environments/Landscape/Materials/Biome/CommonMaterials/MF_HQ_WorldAlignedNormal}.Result) {name=Normal_SM, nodeColor=<rgba 0.5 0 1 1>}
  n25 = MaterialFunctionCall() {materialFunction=/Game/Athena/Helios/Environments/Landscape/Materials/MaterialFunctions_Utilities/MF_ScotchBroom_Colors}
  NamedRerouteDeclaration(None=@n25.Flower1_Color) {name=FlowerColor, nodeColor=<rgba 0.5 0 1 1>}
  n26 = ComponentMask(None=MaterialFunctionCall!(
    Texture=@n22
    UseMipBias
    MipBias
    LowEndTiling
    LandscapeCoordinates=@n13
    HexGridSize=@n15
    HexGridUVInput
    HexGridGradientSharpness=@n14
    MaxScale=@n17
    MinScale=@n16
    MaxOffset
    MinOffset
    "MaxRotation(Degrees)"
    "MinRotation(Degrees)"
    forceLowQuality
   ) {materialFunction=/Game/Athena/Asteria/Environments/Landscape/Materials/MaterialFunctions/UtilityFunctions/Landscape_NormalTextureSample}.Result) {b=true, g=true, r=true}
  NamedRerouteDeclaration(None=StaticSwitch(True=@n26, False=Transform(None=@n26), Value=@n24)) {name=Normal, nodeColor=<rgba 0.5 0 1 1>}
  NamedRerouteDeclaration(None=@n25.FlowerWorldMask1) {name=FlowerWorldMask1, nodeColor=<rgba 0.5 0 1 1>}
  n27 = FunctionInput(Preview=TextureObject() {texture=/Game/Environments/Helios/Terrain/Textures/T_Helios_Chap_CoastGround_01_D}) {bUsePreviewValueAsDefault=true, inputName=Chap_D, inputType=FunctionInput_Texture2D, sortPriority=-400}
  n28 = MaterialFunctionCall(Texture=@n27, LandscapeCoordinates=@n13, HexGridGradientSharpness=@n14, HexGridSize=@n15, MinScale=@n16, MaxScale=@n17) {materialFunction=&Landscape_TextureSample}
  NamedRerouteDeclaration(None=ComponentMask(None=@n28.Result) {a=true}) {name=Grass_ChaparralCoast_Height, nodeColor=<rgba 0.5 0 1 1>}
  NamedRerouteDeclaration(None=ComponentMask(None=@n28.Result) {b=true, g=true, r=true}) {name=Grass_ChaparralCoast_Diffuse, nodeColor=<rgba 0.5 0 1 1>}
  NamedRerouteDeclaration(None=@n25.LandscapeLayerMaskNoise) {name=LandscapeLayerMaskNoise, nodeColor=<rgba 1 0 0.063 1>}
  NamedRerouteDeclaration(None=@n18.R) {name=Specular, nodeColor=<rgba 0.5 0 1 1>}
  NamedRerouteDeclaration(None=@n18.B) {name=Roughness, nodeColor=<rgba 0.5 0 1 1>}
  n29 = TextureSample(UVs=@n19.Result, Tex=@n27) {samplerSource=SSM_Wrap_WorldGroupSettings, texture=&T_Helios_Chap_CoastGround_01_D}
  NamedRerouteDeclaration(None=@n29.A) {name=Grass_ChaparralCoast_Height_SM, nodeColor=<rgba 0.5 0 1 1>}
  NamedRerouteDeclaration(None=@n29.RGB) {name=Grass_ChaparralCoast_Diffuse_SM, nodeColor=<rgba 0.5 0 1 1>}
  NamedRerouteDeclaration(None=@n21.R) {name=Specular_SM, nodeColor=<rgba 0.5 0 1 1>}
  NamedRerouteDeclaration(None=@n21.B) {name=Roughness_SM, nodeColor=<rgba 0.5 0 1 1>}
  n30 = ScalarParameter() {defaultValue=2, group=Displacement, parameterName=Grass_ChaparralCoast_DisplacementPower}
  Saturate(None=Multiply(A=Power(Base=@n11, Exp=@n30), B=ScalarParameter() {defaultValue=2, group=Displacement, parameterName=Grass_ChaparralCoast_DisplacementMod}))
  ScalarParameter() {defaultValue=900, group=Grass_ChaparralCoast, parameterName=Grass_ChaparralCoast_WorldScale}
 ]}
