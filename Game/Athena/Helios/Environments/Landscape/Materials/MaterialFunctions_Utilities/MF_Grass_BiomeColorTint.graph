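# Grass material function for Helios landscapes that tints base color by terrain biome (boreal, coast, desert, Hades, Olympus, Doomstadt and others).
# Mask channels from TerrainMasks QuadB-D drive chained lerps between parameter colors multiplied by base color variants, some desaturated; QuadA is unconnected.
# The main output builds separate body and edge tints, while Result Mobile uses one simpler chain.
# The UseHeliosBIomeColors static switch gates tinting on both outputs.
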MaterialFunction /Game/Athena/Helios/Environments/Landscape/Materials/MaterialFunctions_Utilities/MF_Grass_BiomeColorTint
graph {
 properties {libraryCategoriesText [FN-Landscape]}
 nodes [
  FunctionInput() {bUsePreviewValueAsDefault=true, inputName=TerrainMasks_QuadA, inputType=FunctionInput_Vector4, sortPriority=10}
  n1 = NamedRerouteUsage()
  n2 = NamedRerouteUsage()
  FunctionOutput(None=SetMaterialAttributes!(
    MaterialAttributes=NamedRerouteUsage()
    BaseColor=StaticSwitchParameter(True=@n1, False=@n2) {defaultValue=true, parameterName=UseHeliosBIomeColors}
   ) {attributeSetTypes [#1]})
  n3 = NamedRerouteUsage()
  n4 = NamedRerouteUsage()
  FunctionOutput(None=StaticSwitchParameter(True=SetMaterialAttributes(MaterialAttributes=@n3, BaseColor=@n4) {attributeSetTypes [#1]}, False=@n3) {defaultValue=true, parameterName=UseHeliosBIomeColors}) {outputName=Result Mobile, sortPriority=1}
  n5 = MaterialFunctionCall(Float4=FunctionInput() {bUsePreviewValueAsDefault=true, inputName=TerrainMasks_QuadB, inputType=FunctionInput_Vector4, sortPriority=20}) {materialFunction=&BreakOutFloat4Components}
  NamedRerouteDeclaration(None=@n5.R) {name=Biome_Chaparral_Coast, nodeColor=<rgba 0.5 0 1 1>}
  NamedRerouteDeclaration(None=@n5.G) {name=Biome_Boreal_Forest, nodeColor=<rgba 0.5 0 1 1>}
  n6 = NamedRerouteUsage()
  n7 = VectorParameter() {group=Grass, parameterName=GrassColor_Chaparral_Grassland, defaultValue=<rgba 0.4 0.8 0.8 1>}
  n8 = NamedRerouteDeclaration(None=Saturate(None=Multiply(A=@n6, B=@n7.RGB))) {name=NoBiomeColor, nodeColor=<rgba 0.5 0 1 1>}
  n9 = VectorParameter() {group=BiomeColors, parameterName=GrassColor_Chaparral_CoastZ_B, defaultValue=<rgba 1 1 1 1>}
  n10 = Desaturation(None=NamedRerouteUsage(), Fraction=Constant() {r=0.1})
  n11 = VectorParameter() {group=BiomeColors, parameterName=GrassColor_Chaparral_CoastZ_A, defaultValue=<rgba 1 1 1 1>}
  n12 = OneMinus(None=NamedRerouteUsage())
  n13 = Constant() {r=2}
  n14 = Saturate(None=LinearInterpolate(A=Multiply(A=@n9.RGB, B=@n10), B=Multiply(A=@n11.RGB, B=@n10), Alpha=Power(Base=@n12, Exp=@n13)))
  n15 = NamedRerouteUsage()
  n16 = NamedRerouteUsage()
  n17 = Multiply(A=@n16, B=VectorParameter() {group=BiomeColors, parameterName=GrassColor_Boreal_Tundra, defaultValue=<rgba 0.45 0.5688 2.25 1>}.RGB)
  n18 = NamedRerouteUsage()
  n19 = VectorParameter() {group=BiomeColors, parameterName=GrassColor_Boreal_Forest, defaultValue=<rgba 0.3 0.3792 1.5 1>}
  n20 = LinearInterpolate(
   A=LinearInterpolate(A=LinearInterpolate(A=@n8, B=@n14, Alpha=@n15), B=Saturate(None=@n17), Alpha=NamedRerouteUsage())
   B=Saturate(None=Multiply(A=@n18, B=@n19.RGB))
   Alpha=NamedRerouteUsage()
  )
  n21 = NamedRerouteUsage()
  n22 = Saturate(None=Multiply(A=@n21, B=VectorParameter() {group=BiomeColors, parameterName=GrassColor_Boreal_Mountain, defaultValue=<rgba 0.3 0.3792 1.5 1>}.RGB))
  n23 = NamedRerouteUsage()
  n24 = VectorParameter() {group=BiomeColors, parameterName=GrassColor_Hades01, defaultValue=<rgba 0.5525 0.5526 0.85 1>}
  n25 = VectorParameter() {group=BiomeColors, parameterName=GrassColor_Hades02, defaultValue=<rgba 0.56 1.0373 1.4 1>}
  n26 = NamedRerouteUsage()
  n27 = ComponentMask(None=@n26) {g=true}
  n28 = ScalarParameter() {defaultValue=6.6, group=BiomeColors, parameterName=GrassColor_Hades_Mask_Mod}
  n29 = Saturate(None=Add(A=Multiply(A=@n27, B=@n28), B=ScalarParameter() {defaultValue=-0.75, group=BiomeColors, parameterName=GrassColor_Hades_Mask_Add}))
  n30 = Multiply(A=Saturate(None=LinearInterpolate(A=@n24.RGB, B=@n25.RGB, Alpha=@n29)), B=Desaturation(None=@n26, Fraction=Constant() {r=1}))
  n31 = VectorParameter() {group=BiomeColors, parameterName=GrassColor_Olympus, defaultValue=<rgba 1 0.4349 0.15 4.5>}
  n32 = Desaturation(None=NamedRerouteUsage(), Fraction=ScalarParameter() {defaultValue=0.75, group=BiomeColors, parameterName=GrassColor_Olympus_Desat})
  n33 = Saturate(None=Multiply(A=Multiply(A=@n31.RGB, B=@n32), B=@n31.A))
  n34 = ScalarParameter() {defaultValue=0.25, group=BiomeColors, parameterName=GrassColor_Olympus_DesatNoise}
  n35 = NamedRerouteUsage()
  n36 = NamedRerouteUsage()
  n37 = LinearInterpolate(
   A=LinearInterpolate(A=LinearInterpolate(A=@n20, B=@n22, Alpha=@n23), B=@n30, Alpha=NamedRerouteUsage())
   B=LinearInterpolate(A=@n33, B=Desaturation(None=@n33, Fraction=@n34), Alpha=ComponentMask(None=@n35) {r=true})
   Alpha=@n36
  )
  n38 = Saturate(None=Multiply(A=NamedRerouteUsage(), B=VectorParameter() {group=BiomeColors, parameterName=GrassColor_Desert, defaultValue=<rgba 1.5 0.75 1.1639 1>}.RGB))
  n39 = Max(A=NamedRerouteUsage(), B=NamedRerouteUsage())
  n40 = Multiply(A=NamedRerouteUsage(), B=VectorParameter() {group=BiomeColors, parameterName=GrassColor_DesertRiver, defaultValue=<rgba 0.4 0.8 0.8 1>}.RGB)
  n41 = NamedRerouteUsage()
  n42 = NamedRerouteUsage()
  n43 = VectorParameter() {group=BiomeColors, parameterName=GrassColor_Doomstadt, defaultValue=<rgba 0.15 0.75 0.7424 1>}
  n44 = NamedRerouteUsage()
  n45 = LinearInterpolate(
   A=LinearInterpolate(A=LinearInterpolate(A=@n37, B=@n38, Alpha=@n39), B=Saturate(None=@n40), Alpha=@n41)
   B=Saturate(None=Multiply(A=@n42, B=@n43.RGB))
   Alpha=@n44
  )
  n46 = Saturate(None=Multiply(
    A=NamedRerouteUsage()
    B=VectorParameter() {group=Grass, parameterName=GrassEdgeColor_Chaparral_Grassland, defaultValue=<rgba 2 1.0824 0.5 1>}.RGB
   ))
  n47 = Saturate(None=Multiply(A=@n14, B=VectorParameter() {group=BiomeColors, parameterName=GrassEdgeColor_Chaparral_Coast, defaultValue=<rgba 1 1 1 1>}.RGB))
  n48 = NamedRerouteUsage()
  n49 = Multiply(A=@n16, B=VectorParameter() {group=BiomeColors, parameterName=GrassEdgeColor_Boreal_Tundra, defaultValue=<rgba 0.6204 0.5649 2 1>}.RGB)
  n50 = VectorParameter() {group=BiomeColors, parameterName=GrassEdgeColor_Boreal_Forest, defaultValue=<rgba 0.6204 0.5649 2 1>}
  n51 = LinearInterpolate(
   A=LinearInterpolate(A=LinearInterpolate(A=@n46, B=@n47, Alpha=@n48), B=Saturate(None=@n49), Alpha=NamedRerouteUsage())
   B=Saturate(None=Multiply(A=@n18, B=@n50.RGB))
   Alpha=NamedRerouteUsage()
  )
  n52 = Saturate(None=Multiply(A=@n21, B=VectorParameter() {group=BiomeColors, parameterName=GrassEdgeColor_Boreal_Mountain, defaultValue=<rgba 0.6204 0.5649 2 1>}.RGB))
  n53 = NamedRerouteUsage()
  n54 = Desaturation(
   None=Multiply(A=@n30, B=ScalarParameter() {defaultValue=1, group=BiomeColors, parameterName=GrassColor_Hades_Edge_Mod})
   Fraction=ScalarParameter() {defaultValue=0.5, group=BiomeColors, parameterName=GrassColor_Hades_Edge_Saturation}
  )
  n55 = VectorParameter() {group=BiomeColors, parameterName=GrassEdgeColor_Olympus, defaultValue=<rgba 1 0.4349 0.15 2.7>}
  n56 = Multiply(A=@n32, B=@n55.RGB)
  n57 = LinearInterpolate(
   A=LinearInterpolate(A=LinearInterpolate(A=@n51, B=@n52, Alpha=@n53), B=Saturate(None=@n54), Alpha=NamedRerouteUsage())
   B=Saturate(None=Multiply(A=@n56, B=@n55.A))
   Alpha=@n36
  )
  n58 = Saturate(None=Multiply(A=NamedRerouteUsage(), B=VectorParameter() {group=BiomeColors, parameterName=GrassEdgeColor_Desert, defaultValue=<rgba 2 0.9041 0.4 1>}.RGB))
  n59 = Multiply(
   A=NamedRerouteUsage()
   B=VectorParameter() {group=BiomeColors, parameterName=GrassEdgeColor_DesertRiver, defaultValue=<rgba 2 1.0824 0.5 1>}.RGB
  )
  n60 = VectorParameter() {group=BiomeColors, parameterName=GrassEdgeColor_Doomstadt, defaultValue=<rgba 0.625 0.86 1.25 1>}
  n61 = LinearInterpolate(
   A=LinearInterpolate(A=LinearInterpolate(A=@n57, B=@n58, Alpha=@n39), B=Saturate(None=@n59), Alpha=@n41)
   B=Saturate(None=Multiply(A=@n42, B=@n60.RGB))
   Alpha=@n44
  )
  n62 = NamedRerouteUsage()
  NamedRerouteDeclaration(None=ComponentMask(None=LinearInterpolate(A=@n45, B=@n61, Alpha=@n62)) {b=true, g=true, r=true}) {name=OutputBasecolor, nodeColor=<rgba 0.5 0 1 1>}
  n63 = FunctionInput() {bUsePreviewValueAsDefault=true, inputName=GrassNoise, inputType=FunctionInput_Vector4, sortPriority=50, previewValue=<0 0 0 0>}
  NamedRerouteDeclaration(None=@n63) {name=GrassNoiseVec4, nodeColor=<rgba 0.5 0 1 1>}
  n64 = MaterialFunctionCall(Float4=FunctionInput() {bUsePreviewValueAsDefault=true, inputName=TerrainMasks_QuadC, inputType=FunctionInput_Vector4, sortPriority=30}) {materialFunction=&BreakOutFloat4Components}
  NamedRerouteDeclaration(None=@n64.R) {name=Biome_DesertRiver, nodeColor=<rgba 0.5 0 1 1>}
  n65 = GetMaterialAttributes(None=FunctionInput(Preview=SetMaterialAttributes!(MaterialAttributes)) {bUsePreviewValueAsDefault=true, inputName=Material, inputType=FunctionInput_MaterialAttributes}) {attributeGetTypes [#1, #2]}
  n66 = NamedRerouteDeclaration(None=@n65.BaseColor) {name=Input_BaseColor, nodeColor=<rgba 0.5 0 1 1>}
  NamedRerouteDeclaration(None=Desaturation(None=@n66, Fraction=Constant() {r=0.5})) {name=BaseColor_DesatHalf, nodeColor=<rgba 0.5 0 1 1>}
  NamedRerouteDeclaration(None=@n64.G) {name=Biome_Doomstadt, nodeColor=<rgba 0.5 0 1 1>}
  NamedRerouteDeclaration(None=@n5.B) {name=Biome_Boreal_Tundra, nodeColor=<rgba 0.5 0 1 1>}
  NamedRerouteDeclaration(None=@n5.A) {name=Biome_Boreal_Mountain, nodeColor=<rgba 0.5 0 1 1>}
  NamedRerouteDeclaration(None=FunctionInput(Preview=Constant()) {bUsePreviewValueAsDefault=true, inputName=GrassEdgeMask, inputType=FunctionInput_Scalar, sortPriority=1}) {name=GrassEdgeMask, nodeColor=<rgba 0.5 0 1 1>}
  NamedRerouteDeclaration(None=Desaturation(None=@n66, Fraction=Constant() {r=1})) {name=BaseColorDesat, nodeColor=<rgba 0.5 0 1 1>}
  n67 = MaterialFunctionCall(Float4=FunctionInput() {bUsePreviewValueAsDefault=true, inputName=TerrainMasks_QuadD, inputType=FunctionInput_Vector4, sortPriority=40}) {materialFunction=&BreakOutFloat4Components}
  NamedRerouteDeclaration(None=@n67.R) {name=Biome_Hades, nodeColor=<rgba 0.5 0 1 1>}
  NamedRerouteDeclaration(None=@n67.G) {name=Biome_Olympus, nodeColor=<rgba 0.5 0 1 1>}
  NamedRerouteDeclaration(None=@n65.MaterialAttributes) {name=InputMaterial, nodeColor=<rgba 0.5 0 1 1>}
  NamedRerouteDeclaration(None=@n65.OpacityMask) {name=Heightmap, nodeColor=<rgba 0.5 0 1 1>}
  n68 = VectorParameter() {group=BiomeColors, parameterName=GrassColor_Boreal_Chaparral_Mobile, defaultValue=<rgba 1.6 1.2 1.7 1>}
  n69 = NamedRerouteUsage()
  n70 = LinearInterpolate(A=NamedRerouteUsage(), B=Saturate(None=Multiply(A=@n68.RGB, B=@n69)), Alpha=NamedRerouteUsage())
  n71 = Saturate(None=Multiply(
    A=VectorParameter() {group=BiomeColors, parameterName=GrassColor_Boreal_Tundra, defaultValue=<rgba 0.4 0.5056 2 1>}.RGB
    B=NamedRerouteUsage()
   ))
  n72 = NamedRerouteUsage()
  n73 = Multiply(
   A=VectorParameter() {group=BiomeColors, parameterName=GrassColor_Boreal_Forest, defaultValue=<rgba 0.25 0.316 1.25 1>}.RGB
   B=NamedRerouteUsage()
  )
  n74 = VectorParameter() {group=BiomeColors, parameterName=GrassColor_Boreal_Mountain, defaultValue=<rgba 0.25 0.316 1.25 1>}
  n75 = NamedRerouteUsage()
  n76 = LinearInterpolate(
   A=LinearInterpolate(A=LinearInterpolate(A=@n70, B=@n71, Alpha=@n72), B=Saturate(None=@n73), Alpha=NamedRerouteUsage())
   B=Saturate(None=Multiply(A=@n74.RGB, B=@n75))
   Alpha=NamedRerouteUsage()
  )
  n77 = Saturate(None=Multiply(
    A=NamedRerouteUsage()
    B=VectorParameter() {group=BiomeColors, parameterName=GrassColor_Hades, defaultValue=<rgba 0.5256 0.5276 0.651 1>}.RGB
   ))
  n78 = NamedRerouteUsage()
  n79 = LinearInterpolate(A=NamedRerouteUsage(), B=NamedRerouteUsage()) {constAlpha=0.1}
  n80 = VectorParameter() {group=BiomeColors, parameterName=MobileGrassColor_Olympus, defaultValue=<rgba 3.2 1.5444 0.7 1>}
  n81 = NamedRerouteUsage()
  n82 = VectorParameter() {group=BiomeColors, parameterName=Mobile GrassColor Desert, defaultValue=<rgba 1.6 0.7529 0.7467 1>}
  n83 = LinearInterpolate(
   A=LinearInterpolate(A=LinearInterpolate(A=@n76, B=@n77, Alpha=@n78), B=Multiply(A=@n79, B=@n80.RGB), Alpha=NamedRerouteUsage())
   B=Saturate(None=Multiply(A=@n81, B=@n82.RGB))
   Alpha=Max(A=NamedRerouteUsage(), B=NamedRerouteUsage())
  )
  n84 = Saturate(None=Multiply(
    A=VectorParameter() {group=BiomeColors, parameterName=MobileGrassColor_Doomstadt, defaultValue=<rgba 0.6 0.6924 2 1>}.RGB
    B=NamedRerouteUsage()
   ))
  n85 = NamedRerouteUsage()
  NamedRerouteDeclaration(None=ComponentMask(None=LinearInterpolate(A=@n83, B=@n84, Alpha=@n85)) {b=true, g=true, r=true}) {name=OutputBasecolorMobile, nodeColor=<rgba 0.5 0 1 1>}
  NamedRerouteDeclaration(None=@n67.B) {name=Biome_Arid_Desert, nodeColor=<rgba 0.5 0 1 1>}
  NamedRerouteDeclaration(None=@n67.A) {name=Biome_Open_Desert, nodeColor=<rgba 0.5 0 1 1>}
  ScalarParameter() {defaultValue=2, group=BiomeColors, parameterName=GrassColor_Chaparral_Coast_BaseA_Pow}
 ]
}
