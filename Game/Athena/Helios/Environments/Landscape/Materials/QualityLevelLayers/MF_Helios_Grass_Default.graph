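# Default grass layer material function for the Helios landscape, outputting grass material attributes with displacement.
# Packed grass masks and a dirt normal map, sampled with world-space UVs, drive specular, roughness, normals and a cliff-attenuated height map.
# Procedural variation masks blend tint colors into base color with height-based tips.
# A static switch selects Athena biome colors or terrain-mask tinting; hex-tiled sampling and several alternate blends are left unconnected.
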MaterialFunction /Game/Athena/Helios/Environments/Landscape/Materials/QualityLevelLayers/MF_Helios_Grass_Default
graph {nodes [
  Constant3Vector() {constant=<rgba 0.2396 0.6666 0 1>}
  Constant3Vector() {constant=<rgba 0.9853 1.2 0.3 1>}
  Constant3Vector() {constant=<rgba 0.711 1.0666 0 1>}
  Constant3Vector() {constant=<rgba 0.3797 0.7333 0.0733 1>}
  Constant() {r=945}
  Constant() {r=700}
  Constant() {r=4}
  Constant() {r=10}
  Constant() {r=0.25}
  Constant() {r=-0.6}
  Constant() {r=0.5}
  Constant() {r=2}
  Constant() {r=0.3}
  Constant() {r=1.5}
  Constant() {r=-9}
  Constant() {r=5}
  n1 = NamedRerouteUsage()
  n2 = ScalarParameter() {defaultValue=0.25, group=Displacement, parameterName=Grass_Displacement_Low}
  n3 = ScalarParameter() {defaultValue=0.5, group=Displacement, parameterName=Grass_Displacement_High}
  n4 = SetMaterialAttributes!(
   MaterialAttributes
   BaseColor=NamedRerouteUsage()
   Specular=NamedRerouteUsage()
   Roughness=NamedRerouteUsage()
   Normal=NamedRerouteUsage()
   OpacityMask=@n1
   Displacement=Saturate(None=LinearInterpolate(A=@n2, B=@n3, Alpha=@n1))
  ) {attributeSetTypes [
    #1
    #2
    #3
    #4
    #5
    #6
   ]}
  n5 = NamedRerouteUsage()
  n6 = FunctionInput(Preview=Constant()) {bUsePreviewValueAsDefault=true, inputName=TerrainMasks_QuadA, inputType=FunctionInput_Vector4, sortPriority=-400}
  n7 = FunctionInput(Preview=Constant()) {bUsePreviewValueAsDefault=true, inputName=TerrainMasks_QuadB, inputType=FunctionInput_Vector4, sortPriority=-300}
  n8 = FunctionInput(Preview=Constant()) {bUsePreviewValueAsDefault=true, inputName=TerrainMasks_QuadC, inputType=FunctionInput_Vector4, sortPriority=-200}
  n9 = FunctionInput(Preview=Constant()) {bUsePreviewValueAsDefault=true, inputName=TerrainMasks_QuadD, inputType=FunctionInput_Vector4, sortPriority=-100}
  n10 = NamedRerouteUsage()
  FunctionOutput(None=StaticSwitchParameter(
    True=MaterialFunctionCall!(SetMaterialAttributes=@n4) {materialFunction=/Game/Athena/Helios/Environments/Landscape/Materials/MaterialFunctions_Utilities/MF_Athena_Grass_Biome_Colors}.Blended Material
    False=MaterialFunctionCall!(
     Material=@n4
     GrassEdgeMask=@n5
     TerrainMasks_QuadA=@n6
     TerrainMasks_QuadB=@n7
     TerrainMasks_QuadC=@n8
     TerrainMasks_QuadD=@n9
     GrassNoise=@n10
    ) {materialFunction=/Game/Athena/Helios/Environments/Landscape/Materials/MaterialFunctions_Utilities/MF_Grass_BiomeColorTint}.Result
   ) {parameterName=UseAthenaBiomeColors})
  n11 = MaterialFunctionCall!(Scale=ScalarParameter() {defaultValue=945, group=Grass, parameterName=Grass_WorldAlignedScale, sortPriority=-1000}) {materialFunction=/Game/Athena/Apollo/Environments/Landscape/Materials/Biome/CommonMaterials/MF_SimpleWorldUV}
  n12 = ScalarParameter() {defaultValue=10, group=Grass, parameterName=Grass_WorldAlignedScalePeriodic, sortPriority=-1000}
  n13 = WorldPosition() {worldPositionShaderOffset=WPT_CameraRelative}
  LinearInterpolate(
   A=@n11.Result
   B=Multiply(
    A=MaterialFunctionCall!(Scale=@n12, Camera Relative World Position=@n13.XYZ) {materialFunction=/Engine/Functions/Engine_MaterialFunctions03/Texturing/PeriodicWorldPosition}.XY
    B=Constant2Vector() {g=-1, r=-1}
   )
   Alpha=ScalarParameter() {group=Grass, parameterName=Grass_WorldAlignedScalePeriodic_DEBUG, sortPriority=-1000}
  )
  n14 = VectorParameter() {group=Grass, parameterName=GrassTipColor, defaultValue=<rgba 4 3.5 1 1>}
  n15 = VectorParameter() {group=Grass, parameterName=GrassNoiseColor_01, sortPriority=-20, defaultValue=<rgba 0.0282 0.07 0.0245 1>}
  n16 = VectorParameter() {group=Grass, parameterName=GrassNoiseColor_02, sortPriority=-19, defaultValue=<rgba 0.027 0.09 0.0342 1>}
  n17 = NamedRerouteUsage()
  n18 = VectorParameter() {group=Grass, parameterName=GrassNoiseColor_03, sortPriority=-18, defaultValue=<rgba 0.064 0.08 0.048 1>}
  n19 = VectorParameter() {group=Grass, parameterName=GrassNoiseColor_04, sortPriority=-17, defaultValue=<rgba 0.0419 0.125 0.0375 1>}
  n20 = LinearInterpolate(
   A=LinearInterpolate(A=LinearInterpolate(A=@n15.RGB, B=@n16.RGB, Alpha=@n17), B=@n18.RGB, Alpha=NamedRerouteUsage())
   B=@n19.RGB
   Alpha=NamedRerouteUsage()
  )
  n21 = ScalarParameter() {defaultValue=-0.25, group=Grass, parameterName=GrassTipBlendMin}
  n22 = ScalarParameter() {defaultValue=1, group=Grass, parameterName=GrassTipBlendMax}
  n23 = LinearInterpolate(A=NamedRerouteUsage(), B=LinearInterpolate(A=Constant() {r=0.75}, B=Constant(), Alpha=NamedRerouteUsage()), Alpha=Constant() {r=0.5}) {desc=This makes the transition softer}
  LinearInterpolate(A=@n20, B=Saturate(None=Multiply(A=@n20, B=@n14.RGB)), Alpha=Saturate(None=LinearInterpolate(A=@n21, B=@n22, Alpha=@n23)))
  n24 = ScalarParameter() {defaultValue=2, group=Displacement, parameterName=Grass_DisplacementMod}
  LinearInterpolate(
   B=Saturate(None=Multiply(A=@n1, B=@n24))
   Alpha=ScalarParameter() {defaultValue=0.25, group=Displacement, parameterName=Grass_DisplacementAmt}
  ) {constA=0.5}
  n25 = MaterialFunctionCall!(Coordinates, R_01_Scale, R_02_Scale, R_03_Scale, R_04_Scale, G_01_Scale, G_02_Scale, G_03_Scale, G_04_Scale) {materialFunction=/Game/Athena/Helios/Environments/Landscape/Materials/MaterialFunctions_Utilities/MF_VariationMasks2}
  MaterialFunctionCall(X=@n25.G_01, Y=@n25.G_01, Z=@n25.G_01, A=@n25.G_01) {materialFunction=&MakeFloat4}
  n26 = TextureObject() {samplerType=SAMPLERTYPE_Masks, texture=/Game/Environments/Apollo/Sets/Terrain/Textures/T_Apollo_Grass_D_02}
  n27 = Divide(
   A=WorldPosition() {worldPositionShaderOffset=WPT_ExcludeAllShaderOffsets}.XY
   B=ScalarParameter() {defaultValue=700, group=Grass, parameterName=Grass_WorldAlignedHexBasedScale, sortPriority=-1000}
  )
  n28 = StaticBool()
  MaterialFunctionCall!(
   Texture=@n26
   LandscapeCoordinates=@n27
   LowEndTiling
   HexGridGradientSharpness
   HexGridSize
   HexGridUVInput
   MinScale
   MaxScale
   MinOffset
   MaxOffset
   "MaxRotation(Degrees)"
   "MinRotation(Degrees)"
   forceLowQuality=@n28
  ) {materialFunction=/Game/Athena/Asteria/Environments/Landscape/Materials/MaterialFunctions/UtilityFunctions/Landscape_TextureSample}
  NamedRerouteDeclaration(None=MaterialFunctionCall(X=@n25.R_01, Y=@n25.R_01, Z=@n25.R_01, A=@n25.R_01) {materialFunction=&MakeFloat4}.Result) {name=GrassNoiseOut, nodeColor=<rgba 0.5 0 1 1>}
  n29 = MaterialFunctionCall(
   Transition Phase=MaterialFunctionCall() {materialFunction=/Game/Athena/Helios/Environments/Landscape/Materials/MaterialFunctions_Utilities/MF_CliffSlopeMask}.Result
   Height Texture=Constant() {r=0.5}
  ) {materialFunction=&HeightLerp}
  n30 = Constant() {r=3}
  n31 = Saturate(None=Multiply(A=Power(Base=@n29.Lerp Alpha No Contrast, Exp=@n30), B=Constant() {r=6}))
  NamedRerouteDeclaration(None=Multiply(A=OneMinus(None=@n31), B=NamedRerouteUsage())) {name=GrassHeightMap, nodeColor=<rgba 0.5 0 1 1>}
  n32 = TextureSample(UVs=@n11.Result, Tex=@n26) {samplerSource=SSM_Wrap_WorldGroupSettings, samplerType=SAMPLERTYPE_Masks, texture=&T_WhiteMask}
  n33 = MaterialFunctionCall(Float4=@n32.RGBA) {materialFunction=&BreakOutFloat4Components}
  NamedRerouteDeclaration(None=@n33.G) {name=GrassSpecularMask, nodeColor=<rgba 0.5 0 1 1>}
  n34 = ScalarParameter() {defaultValue=0.1, group=Grass, parameterName=GrassMinSpec}
  n35 = ScalarParameter() {defaultValue=1, group=Grass, parameterName=GrassMaxSpec}
  n36 = Saturate(None=LinearInterpolate(
    A=ScalarParameter() {group=Grass, parameterName=GrassSBlendMin}
    B=ScalarParameter() {defaultValue=1, group=Grass, parameterName=GrassSBlendMax}
    Alpha=NamedRerouteUsage()
   ))
  NamedRerouteDeclaration(None=Saturate(None=LinearInterpolate(A=@n34, B=@n35, Alpha=@n36))) {name=GrassSpecular, nodeColor=<rgba 0.5 0 1 1>}
  n37 = Constant() {r=0.01}
  n38 = NamedRerouteUsage()
  n39 = LinearInterpolate(A=Constant(), B=Constant() {r=1.5}, Alpha=@n25.G_01)
  n40 = ScalarParameter() {defaultValue=1.5, group=Grass, parameterName=Grass2_Noise2_A}
  n41 = ScalarParameter() {defaultValue=-9, group=Grass, parameterName=Grass2_Noise2_B}
  n42 = LinearInterpolate(
   A=LinearInterpolate(
    A=VectorParameter() {parameterName=G2_C1, defaultValue=<rgba 0.8038 1.0666 0.1066 1>}.RGB
    B=VectorParameter() {parameterName=G2_C2, defaultValue=<rgba 0.6197 0.7 0.1692 1>}.RGB
    Alpha=Saturate(None=@n39)
   )
   B=VectorParameter() {parameterName=G2_C3, defaultValue=<rgba 0.8177 1.0666 0.32 1>}.RGB
   Alpha=Saturate(None=LinearInterpolate(A=@n40, B=@n41, Alpha=@n25.G_02))
  )
  n43 = VectorParameter() {parameterName=G2_C4, defaultValue=<rgba 1.2 0.8869 0.3196 1>}
  n44 = Saturate(None=LinearInterpolate(
    A=ScalarParameter() {defaultValue=-0.25, group=Grass, parameterName=Grass2_Noise3_A}
    B=ScalarParameter() {defaultValue=1.5, group=Grass, parameterName=Grass2_Noise3_B}
    Alpha=@n25.G_03
   ))
  n45 = Multiply(A=@n44, B=ScalarParameter() {defaultValue=0.65, group=Grass, parameterName=Grass2_Noise3_Mod})
  n46 = Multiply(A=@n25.G_04, B=ScalarParameter() {defaultValue=4, group=Grass, parameterName=Grass2_Noise4_Mod})
  n47 = Multiply(
   A=LinearInterpolate(
    A=VectorParameter() {parameterName=G2_C5, defaultValue=<rgba 0.4785 0.8 0.2 1>}.RGB
    B=LinearInterpolate(A=@n42, B=@n43.RGB, Alpha=@n45)
    Alpha=Saturate(None=@n46)
   )
   B=VectorParameter() {group=Grass, parameterName=G2_ColorAdjust, defaultValue=<rgba 0.0954 0.125 0.0625 1>}.RGB
  )
  n48 = Saturate(None=Add(A=Multiply(A=@n37, B=@n38), B=Multiply(A=@n38, B=@n47)))
  n49 = Multiply(A=@n48, B=VectorParameter() {group=Grass, parameterName=Grass2_TipColor, defaultValue=<rgba 1.75 1.5695 1.4 1>}.RGB)
  n50 = NamedRerouteUsage()
  n51 = ScalarParameter() {defaultValue=0.7, group=Grass, parameterName=Grass2_Tip_HeightPower}
  NamedRerouteDeclaration(None=LinearInterpolate(A=@n48, B=Saturate(None=@n49), Alpha=Power(Base=@n50, Exp=@n51))) {name=GrassBaseColor, nodeColor=<rgba 0.5 0 1 1>}
  NamedRerouteDeclaration(None=@n33.B) {name=GrassRoughnessMask, nodeColor=<rgba 0.5 0 1 1>}
  n52 = Multiply(A=@n44, B=Constant() {r=0.15})
  n53 = LinearInterpolate(
   A=ScalarParameter() {defaultValue=-0.15, group=Grass, parameterName=GrassRBlendMin}
   B=ScalarParameter() {defaultValue=1, group=Grass, parameterName=GrassRBlendMax}
   Alpha=NamedRerouteUsage()
  )
  n54 = Saturate(None=LinearInterpolate(
    A=ScalarParameter() {defaultValue=0.8, group=Grass, parameterName=GrassRoughness}
    B=ScalarParameter() {defaultValue=0.6, group=Grass, parameterName=GrassMaxRoughness}
    Alpha=Saturate(None=@n53)
   ))
  NamedRerouteDeclaration(None=Saturate(None=Add(A=@n52, B=@n54))) {name=GrassRoughness, nodeColor=<rgba 0.5 0 1 1>}
  n55 = MaterialFunctionCall!(NoiseTexture, Position, MicroScale_01, MicroScale_02, MicroScale_03, MacroScale) {materialFunction=/Game/Athena/Helios/Environments/Landscape/Materials/MaterialFunctions_Utilities/MF_VariationMasks}
  NamedRerouteDeclaration(None=@n55.MicroVariationMask_01) {name=GrassColorMask_01, nodeColor=<rgba 0.5 0 1 1>}
  NamedRerouteDeclaration(None=@n55.MicroVariationMask_02) {name=GrassColorMask_02, nodeColor=<rgba 0.5 0 1 1>}
  NamedRerouteDeclaration(None=@n55.MicroVariationMask_03) {name=GrassColorMask_03, nodeColor=<rgba 0.5 0 1 1>}
  n56 = Constant()
  n57 = Constant()
  n58 = VertexNormalWS()
  n59 = Power(Base=Abs(None=ComponentMask(None=@n58) {b=true}), Exp=ScalarParameter() {defaultValue=1.5, group=Grass, parameterName=GrassSlope1_Pow})
  n60 = Max(
   A=Multiply(
    A=FunctionInput(Preview=@n57) {bUsePreviewValueAsDefault=true, inputName=Grass_EdgeMask, inputType=FunctionInput_Scalar, sortPriority=16}
    B=ScalarParameter() {defaultValue=2, group=Grass, parameterName=GrassEdgeSpread}
   )
   B=Multiply(A=OneMinus(None=@n59), B=ScalarParameter() {defaultValue=1.25, group=Grass, parameterName=GrassSlope1_Mod})
  )
  n61 = FunctionInput(Preview=StaticBool()) {
   bUsePreviewValueAsDefault=true
   desc=Setting this to true will ignore Grass_EdgeMask input
   description=Setting this to true will ignore Grass_EdgeOnly input
   inputName=UseCustomSlopeMask
   inputType=FunctionInput_StaticBool
   sortPriority=13
  }
  NamedRerouteDeclaration(None=StaticSwitch(
    True=FunctionInput(Preview=@n56) {bUsePreviewValueAsDefault=true, inputName=CustomSlopeMask, inputType=FunctionInput_Scalar, sortPriority=14}
    False=Saturate(None=@n60)
    Value=@n61
   )) {name=GrassEdgeSlopeMask, nodeColor=<rgba 0.5 0 1 1>}
  n62 = TextureObject() {samplerType=SAMPLERTYPE_Normal, texture=/Game/Environments/Apollo/Sets/Terrain/Textures/T_Apollo_Dirt_N_02b}
  n63 = TextureSample(UVs=@n11.Result, Tex=@n62) {samplerSource=SSM_Wrap_WorldGroupSettings, samplerType=SAMPLERTYPE_Normal, texture=&FlatNormalBlue}
  n64 = FunctionInput(Preview=StaticBool()) {
   bUsePreviewValueAsDefault=true
   description="Default = False"
   inputName=UseTangentNormals
   inputType=FunctionInput_StaticBool
   sortPriority=1
  }
  n65 = ScalarParameter() {defaultValue=0.25, group=Grass, parameterName=GrassNormalAmount}
  NamedRerouteDeclaration(None=MaterialFunctionCall!(Normals=@n63.RGB, TransformToTangentSpace=@n64, Normal Strength=@n65) {materialFunction=/Game/Athena/Apollo/Environments/Landscape/Materials/Biome/CommonMaterials/MF_HQ_WorldAlignedNormal}.Result) {name=BladeNormals, nodeColor=<rgba 0.5 0 1 1>}
  NamedRerouteDeclaration(None=@n33.A) {name=HeightMapFromTexture, nodeColor=<rgba 0.5 0 1 1>}
  NamedRerouteDeclaration(None=@n33.R) {name=BaseColorDetail, nodeColor=<rgba 1 0.0703 0 1>}
  NamedRerouteUsage()
  OneMinus(None=FunctionInput(Preview=Constant()) {bUsePreviewValueAsDefault=true, inputName=LayerWeight, inputType=FunctionInput_Scalar, sortPriority=-500})
  Power(Base=@n1, Exp=ScalarParameter() {defaultValue=1, group=Displacement, parameterName=Grass_DisplacementPower})
  Saturate(None=ScalarParameter() {defaultValue=0.5, group=Displacement, parameterName=Grass_DisplacementFlatValue})
  Saturate(None=LinearInterpolate(
    A=ScalarParameter() {group=Displacement, parameterName=Grass_HMTest_Low}
    B=ScalarParameter() {defaultValue=1, group=Displacement, parameterName=Grass_HMTest_High}
    Alpha=@n1
   ))
  n66 = LinearInterpolate(
   A=Constant3Vector() {constant=<rgba 0 0 1 0>}
   B=ComponentMask(None=MaterialFunctionCall!(
     Texture=@n62
     UseMipBias
     MipBias
     LowEndTiling
     LandscapeCoordinates=@n27
     HexGridSize
     HexGridUVInput
     HexGridGradientSharpness
     MaxScale
     MinScale
     MaxOffset
     MinOffset
     "MaxRotation(Degrees)"
     "MinRotation(Degrees)"
     forceLowQuality=@n28
    ) {materialFunction=/Game/Athena/Asteria/Environments/Landscape/Materials/MaterialFunctions/UtilityFunctions/Landscape_NormalTextureSample}.Result) {b=true, g=true, r=true}
   Alpha=@n65
  )
  StaticSwitch(True=@n66, False=Transform(None=@n66), Value=@n64)
 ]}
