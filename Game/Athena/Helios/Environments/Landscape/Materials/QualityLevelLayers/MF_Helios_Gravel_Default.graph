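# Material function for the default-quality Helios landscape gravel layer, outputting a biome-tinted material driven by terrain masks plus an untinted ForMudMaterial variant.
# Gravel, mud and large rock-dirt textures, sampled in world UVs at separate scales, combine through chained height blends, with noise masks for variation.
# A LayerWeight height lerp tints gravel color, and Gravel_DisplacementAmt scales displacement.
# Constants mirror the unconnected Gravel-group parameters.
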MaterialFunction /Game/Athena/Helios/Environments/Landscape/Materials/QualityLevelLayers/MF_Helios_Gravel_Default
graph {nodes [
  n1 = Multiply(A=LandscapeLayerCoords(), B=NamedRerouteUsage())
  ComponentMask(None=MaterialFunctionCall!(
    Texture=TextureObject() {texture=/Game/Environments/Helios/Terrain/Textures/T_Helios_Gravel_01_D}
    LandscapeCoordinates=@n1
    LowEndTiling=NamedRerouteUsage()
    HexGridGradientSharpness=NamedRerouteUsage()
    HexGridSize=NamedRerouteUsage()
    HexGridUVInput
    MinScale=NamedRerouteUsage()
    MaxScale=NamedRerouteUsage()
    MinOffset=NamedRerouteUsage()
    MaxOffset=NamedRerouteUsage()
    "MaxRotation(Degrees)"=NamedRerouteUsage()
    "MinRotation(Degrees)"=NamedRerouteUsage()
    forceLowQuality=NamedRerouteUsage()
   ) {materialFunction=/Game/Athena/Asteria/Environments/Landscape/Materials/MaterialFunctions/UtilityFunctions/Landscape_TextureSample}.Result) {a=true}
  n2 = SetMaterialAttributes!(
   MaterialAttributes
   BaseColor=NamedRerouteUsage()
   Specular=NamedRerouteUsage()
   Roughness=NamedRerouteUsage()
   Normal=NamedRerouteUsage()
   OpacityMask=NamedRerouteUsage()
   Displacement=NamedRerouteUsage()
  ) {
   desc=HighEnd Material
   attributeSetTypes [
    #1
    #2
    #3
    #4
    #5
    #6
   ]
  }
  n3 = Constant()
  n4 = Constant()
  n5 = Constant()
  n6 = Constant()
  FunctionOutput(None=MaterialFunctionCall!(
    Material=@n2
    TerrainMasks_QuadA=FunctionInput(Preview=@n3) {bUsePreviewValueAsDefault=true, inputName=TerrainMasks_QuadA, inputType=FunctionInput_Vector4, sortPriority=-400}
    TerrainMasks_QuadB=FunctionInput(Preview=@n4) {bUsePreviewValueAsDefault=true, inputName=TerrainMasks_QuadB, inputType=FunctionInput_Vector4, sortPriority=-300}
    TerrainMasks_QuadC=FunctionInput(Preview=@n5) {bUsePreviewValueAsDefault=true, inputName=TerrainMasks_QuadC, inputType=FunctionInput_Vector4, sortPriority=-200}
    TerrainMasks_QuadD=FunctionInput(Preview=@n6) {bUsePreviewValueAsDefault=true, inputName=TerrainMasks_QuadD, inputType=FunctionInput_Vector4, sortPriority=-100}
   ) {materialFunction=/Game/Athena/Helios/Environments/Landscape/Materials/MaterialFunctions_Utilities/MF_Gravel_BiomeColorTint}.Result)
  n7 = SetMaterialAttributes!(
   MaterialAttributes
   BaseColor=NamedRerouteUsage()
   Specular=NamedRerouteUsage()
   Roughness=NamedRerouteUsage()
   Normal=NamedRerouteUsage()
   OpacityMask=NamedRerouteUsage()
   Displacement=NamedRerouteUsage()
   Opacity=NamedRerouteUsage()
  ) {
   desc=Material only
   attributeSetTypes [
    #1
    #2
    #3
    #4
    #5
    #6
    #7
   ]
  }
  FunctionOutput(None=@n7) {outputName=ForMudMaterial, sortPriority=10}
  n8 = Multiply(A=LandscapeLayerCoords(), B=NamedRerouteUsage())
  MaterialFunctionCall(
   Texture=TextureObject() {samplerType=SAMPLERTYPE_Normal, texture=/Game/Environments/Helios/Terrain/Textures/T_Helios_Gravel_01_N}
   LandscapeCoordinates=@n8
   LowEndTiling=NamedRerouteUsage()
   HexGridGradientSharpness=NamedRerouteUsage()
   HexGridSize=NamedRerouteUsage()
   MinScale=NamedRerouteUsage()
   MaxScale=NamedRerouteUsage()
   MinOffset=NamedRerouteUsage()
   MaxOffset=NamedRerouteUsage()
   "MaxRotation(Degrees)"=NamedRerouteUsage()
   "MinRotation(Degrees)"=NamedRerouteUsage()
   forceLowQuality=NamedRerouteUsage()
  ) {materialFunction=&Landscape_TextureSample}
  n9 = Multiply(A=LandscapeLayerCoords(), B=NamedRerouteUsage())
  MaterialFunctionCall(Float3=MaterialFunctionCall(
    Texture=TextureObject() {samplerType=SAMPLERTYPE_Masks, texture=/Game/Environments/Helios/Terrain/Textures/T_Helios_Gravel_01_S}
    LandscapeCoordinates=@n9
    LowEndTiling=NamedRerouteUsage()
    HexGridGradientSharpness=NamedRerouteUsage()
    HexGridSize=NamedRerouteUsage()
    MinScale=NamedRerouteUsage()
    MaxScale=NamedRerouteUsage()
    MinOffset=NamedRerouteUsage()
    MaxOffset=NamedRerouteUsage()
    "MaxRotation(Degrees)"=NamedRerouteUsage()
    "MinRotation(Degrees)"=NamedRerouteUsage()
    forceLowQuality=NamedRerouteUsage()
   ) {materialFunction=&Landscape_TextureSample}.Result) {materialFunction=&BreakOutFloat3Components}
  n10 = Constant3Vector() {constant=<rgba 0.6 0.56 0.48 1>}
  n11 = NamedRerouteUsage()
  n12 = FunctionInput(Preview=Constant()) {bUsePreviewValueAsDefault=true, inputName=LayerWeight, inputType=FunctionInput_Scalar, sortPriority=-500}
  n13 = Saturate(None=Multiply(A=NamedRerouteUsage(), B=Constant() {r=1.25}))
  n14 = Constant() {r=0.5}
  NamedRerouteDeclaration(None=LinearInterpolate(
    A=Multiply(A=@n10, B=@n11)
    B=@n11
    Alpha=MaterialFunctionCall(Transition Phase=@n12, Height Texture=@n13, Contrast=@n14) {materialFunction=&HeightLerp}.Alpha
   )) {name=GravelColor, nodeColor=<rgba 0.0312 0 1 1>}
  n15 = Constant() {r=0.5}
  n16 = NamedRerouteUsage()
  n17 = OneMinus(None=@n16)
  n18 = Multiply(A=NamedRerouteUsage(), B=Multiply(A=@n17, B=@n17))
  n19 = Multiply(A=Saturate(None=Add(A=@n15, B=@n16) {constB=0.5}), B=Saturate(None=Add(A=@n16, B=@n18)))
  n20 = Constant() {r=1.2}
  NamedRerouteDeclaration(None=Saturate(None=Multiply(A=@n19, B=@n20) {constB=2})) {name=AtoB_Noise, nodeColor=<rgba 0.5 0 1 1>}
  n21 = MaterialFunctionCall!(
   Coordinates
   R_01_Scale=Constant() {r=2100}
   R_02_Scale=Constant() {r=3600}
   R_03_Scale=Constant() {r=1200}
   R_04_Scale
   G_01_Scale
   G_02_Scale
   G_03_Scale
   G_04_Scale
  ) {materialFunction=/Game/Athena/Helios/Environments/Landscape/Materials/MaterialFunctions_Utilities/MF_VariationMasks2}
  NamedRerouteDeclaration(None=@n21.R_02) {name=Gravel_Noise2, nodeColor=<rgba 0.5 0 1 1>}
  NamedRerouteDeclaration(None=@n21.R_03) {name=Gravel_Noise3, nodeColor=<rgba 0.5 0 1 1>}
  NamedRerouteDeclaration(None=@n21.R_01) {name=Gravel_Noise1, nodeColor=<rgba 0.5 0 1 1>}
  NamedRerouteDeclaration(None=NamedRerouteUsage()) {name=GravelSpec, nodeColor=<rgba 0 0.6953 1 1>}
  NamedRerouteDeclaration(None=Constant() {r=1}) {name=HexGridSize, nodeColor=<rgba 1 0.5391 0 1>}
  n22 = MaterialFunctionCall!(
   Diffuse 1=NamedRerouteUsage()
   Diffuse 2=NamedRerouteUsage()
   Transistion Phase=NamedRerouteUsage()
   Height Texture 1=NamedRerouteUsage()
   Height Texture 2=NamedRerouteUsage()
   Contrast=Constant() {r=0.5}
  ) {materialFunction=/Engine/Functions/Engine_MaterialFunctions02/Texturing/HeightLerpWithTwoHeightMaps}
  n23 = NamedRerouteUsage()
  n24 = MaterialFunctionCall(
   Diffuse 1=@n22.Results
   Diffuse 2=NamedRerouteUsage()
   Transistion Phase=Multiply(A=NamedRerouteUsage(), B=@n23)
   Height Texture 1=@n22.Results
   Height Texture 2=@n23
  ) {materialFunction=&HeightLerpWithTwoHeightMaps}
  NamedRerouteDeclaration(None=@n24.Resulting Height Map) {name=GravelHeight, nodeColor=<rgba 0.5 0 1 1>}
  NamedRerouteDeclaration(None=Constant() {r=2}) {name=HexGridGradientSharpness, nodeColor=<rgba 0 0.7422 1 1>}
  NamedRerouteDeclaration(None=@n1) {name=LandscapeUV, nodeColor=<rgba 0.7812 1 0 1>}
  NamedRerouteDeclaration(None=Constant() {r=4}) {name=MinScale, nodeColor=<rgba 1 0.8906 0 1>}
  NamedRerouteDeclaration(None=Constant() {r=8}) {name=MaxScale, nodeColor=<rgba 1 0.8906 0 1>}
  NamedRerouteDeclaration(None=Constant() {r=-1}) {name=MinOffset, nodeColor=<rgba 0.7812 1 0 1>}
  NamedRerouteDeclaration(None=Constant() {r=1}) {name=MaxOffset, nodeColor=<rgba 0.7812 1 0 1>}
  NamedRerouteDeclaration(None=StaticBool()) {name=ForceLow, nodeColor=<rgba 0 1 0.3203 1>}
  NamedRerouteDeclaration(None=Constant()) {name=MinRotation, nodeColor=<rgba 0.5938 1 0 1>}
  NamedRerouteDeclaration(None=@n24.Results) {name=DiffuseTextureColor, nodeColor=<rgba 0.5 0 1 1>}
  NamedRerouteDeclaration(None=Constant() {r=0.15}) {name=LowEndTiling, nodeColor=<rgba 0 0.0156 1 1>}
  NamedRerouteDeclaration(None=Constant()) {name=MaxRotation, nodeColor=<rgba 0.5938 1 0 1>}
  NamedRerouteDeclaration(None=Constant() {r=2}) {name=MacroScale, nodeColor=<rgba 0 1 0.1328 1>}
  NamedRerouteDeclaration(None=StaticBool()) {name=useMipBias, nodeColor=<rgba 0 1 0.9531 1>}
  NamedRerouteDeclaration(None=Constant() {r=3}) {name=MipBiasValue, nodeColor=<rgba 0.6641 1 0 1>}
  NamedRerouteDeclaration(None=NamedRerouteUsage()) {name=GravelRoughness, nodeColor=<rgba 0.125 0 1 1>}
  n25 = MaterialFunctionCall!(Scale=Constant() {r=592}) {materialFunction=/Game/Athena/Apollo/Environments/Landscape/Materials/Biome/CommonMaterials/MF_SimpleWorldUV}
  n26 = TextureSample(UVs=@n25.Result) {samplerSource=SSM_Wrap_WorldGroupSettings, texture=/Game/Environments/Asteria/Terrain/Textures/T_Asteria_Mud_01_D}
  NamedRerouteDeclaration(None=@n26.A) {name=GravelB_Height, nodeColor=<rgba 1 0.5 0 1>}
  n27 = TextureSample(UVs=@n25.Result) {
   samplerSource=SSM_Wrap_WorldGroupSettings
   samplerType=SAMPLERTYPE_Normal
   texture=/Game/Environments/Asteria/Terrain/Textures/T_Asteria_Mud_01_N
  }
  NamedRerouteDeclaration(None=@n27.RGB) {name=GravelB_Normal, nodeColor=<rgba 1 0.5 0 1>}
  NamedRerouteDeclaration(None=@n26.RGB) {name=GravelB_Diffuse, nodeColor=<rgba 1 0.5 0 1>}
  n28 = MaterialFunctionCall(Scale=Constant() {r=897}) {materialFunction=&MF_SimpleWorldUV}
  n29 = TextureSample(UVs=@n28.Result) {samplerSource=SSM_Wrap_WorldGroupSettings, texture=/Game/Environments/Asteria/Terrain/Textures/T_Asteria_LargeRockDirt_01_D}
  n30 = Multiply(A=Multiply(A=@n29.A, B=@n29.A), B=Constant() {r=1.25})
  NamedRerouteDeclaration(None=LinearInterpolate(A=@n29.A, B=Saturate(None=@n30), Alpha=NamedRerouteUsage())) {name=GravelC_Height, nodeColor=<rgba 0 1 0 1>}
  NamedRerouteDeclaration(None=@n8) {name=LandscapeUV_1, nodeColor=<rgba 0.7812 1 0 1>}
  n31 = Multiply(A=Constant3Vector() {constant=<rgba 1.3 0.9056 0.52 1>}, B=@n29.RGB)
  NamedRerouteDeclaration(None=LinearInterpolate(A=Saturate(None=@n31), B=@n29.RGB, Alpha=NamedRerouteUsage())) {name=GravelC_Diffuse, nodeColor=<rgba 0 1 0 1>}
  n32 = TextureSample(UVs=@n28.Result) {
   samplerSource=SSM_Wrap_WorldGroupSettings
   samplerType=SAMPLERTYPE_Normal
   texture=/Game/Environments/Asteria/Terrain/Textures/T_Asteria_LargeRockDirt_01_N
  }
  NamedRerouteDeclaration(None=@n32.RGB) {name=GravelC_Normal, nodeColor=<rgba 0 1 0 1>}
  n33 = MaterialFunctionCall(Scale=Constant() {r=533}) {materialFunction=&MF_SimpleWorldUV}
  n34 = TextureSample(UVs=@n33.Result) {
   samplerSource=SSM_Wrap_WorldGroupSettings
   samplerType=SAMPLERTYPE_Normal
   texture=&T_Helios_Gravel_01_N
  }
  NamedRerouteDeclaration(None=@n34.RGB) {name=GravelA_Normal, nodeColor=<rgba 0 0.5 1 1>}
  n35 = TextureSample(UVs=@n33.Result) {samplerSource=SSM_Wrap_WorldGroupSettings, texture=&T_Helios_Gravel_01_D}
  NamedRerouteDeclaration(None=@n35.A) {name=GravelA_Height, nodeColor=<rgba 0 0.5 1 1>}
  NamedRerouteDeclaration(None=@n35.RGB) {name=GravelA_Diffuse, nodeColor=<rgba 0 0.5 1 1>}
  n36 = TextureSample(UVs=@n28.Result) {
   samplerSource=SSM_Wrap_WorldGroupSettings
   samplerType=SAMPLERTYPE_Masks
   texture=/Game/Environments/Asteria/Terrain/Textures/T_Asteria_LargeRockDirt_01_S
  }
  NamedRerouteDeclaration(None=@n36.RGB) {name=GravelC_S, nodeColor=<rgba 0 1 0 1>}
  n37 = TextureSample(UVs=@n25.Result) {
   samplerSource=SSM_Wrap_WorldGroupSettings
   samplerType=SAMPLERTYPE_Masks
   texture=/Game/Environments/Asteria/Terrain/Textures/T_Asteria_Mud_01_S
  }
  NamedRerouteDeclaration(None=@n37.RGB) {name=GravelB_S, nodeColor=<rgba 1 0.5 0 1>}
  n38 = TextureSample(UVs=@n33.Result) {
   samplerSource=SSM_Wrap_WorldGroupSettings
   samplerType=SAMPLERTYPE_Masks
   texture=&T_Helios_Gravel_01_S
  }
  NamedRerouteDeclaration(None=@n38.RGB) {name=GravelA_S, nodeColor=<rgba 0 0.5 1 1>}
  n39 = NamedRerouteUsage()
  n40 = LinearInterpolate(A=NamedRerouteUsage(), B=NamedRerouteUsage(), Alpha=NamedRerouteUsage())
  n41 = NamedRerouteUsage()
  n42 = FunctionInput(Preview=StaticBool()) {
   bUsePreviewValueAsDefault=true
   description="Default = False"
   inputName=UseTangentNormals
   inputType=FunctionInput_StaticBool
   sortPriority=1
  }
  NamedRerouteDeclaration(None=MaterialFunctionCall!(Normals=LinearInterpolate(A=@n39, B=@n40, Alpha=@n41), TransformToTangentSpace=@n42, Normal Strength) {materialFunction=/Game/Athena/Apollo/Environments/Landscape/Materials/Biome/CommonMaterials/MF_HQ_WorldAlignedNormal}.Result) {name=Normals, nodeColor=<rgba 0.5 0 1 1>}
  n43 = NamedRerouteUsage()
  n44 = NamedRerouteUsage()
  n45 = NamedRerouteUsage()
  n46 = MaterialFunctionCall(Float3=LinearInterpolate(A=NamedRerouteUsage(), B=LinearInterpolate(A=@n43, B=@n44, Alpha=@n45), Alpha=NamedRerouteUsage())) {materialFunction=&BreakOutFloat3Components}
  NamedRerouteDeclaration(None=@n46.R) {name=Specular, nodeColor=<rgba 0.5 0 1 1>}
  NamedRerouteDeclaration(None=@n46.B) {name=Roughness, nodeColor=<rgba 0.5 0 1 1>}
  NamedRerouteDeclaration(None=@n9) {name=LandscapeUV_2, nodeColor=<rgba 0.7812 1 0 1>}
  NamedRerouteDeclaration(None=LinearInterpolate(B=NamedRerouteUsage(), Alpha=ScalarParameter() {defaultValue=1, group=Displacement, parameterName=Gravel_DisplacementAmt}) {constA=0.5}) {name=Displacement, nodeColor=<rgba 0.125 0 1 1>}
  NamedRerouteDeclaration(None=@n22.Alpha) {name=AtoB_Mask, nodeColor=<rgba 0.5 0 1 1>}
  NamedRerouteDeclaration(None=@n24.Alpha) {name=ABtoC_Mask, nodeColor=<rgba 0.5 0 1 1>}
  ScalarParameter() {group=Gravel, parameterName=Gravel_ABtoC_BlendContrast}
  ScalarParameter() {defaultValue=2100, group=Gravel, parameterName=Gravel_Noise1_Scale}
  ScalarParameter() {defaultValue=3600, group=Gravel, parameterName=Gravel_Noise2_Scale}
  ScalarParameter() {defaultValue=533, group=Gravel, parameterName=GravelA_WorldScale}
  ScalarParameter() {defaultValue=0.5, group=Gravel, parameterName=GravelEdge_Contrast}
  ScalarParameter() {defaultValue=1200, group=Gravel, parameterName=Gravel_Noise3_Scale}
  ScalarParameter() {defaultValue=1.25, group=Gravel, parameterName=GravelEdge_HeightMod}
  ScalarParameter() {defaultValue=1.25, group=Gravel, parameterName=GravelC_Height_Mod}
  ScalarParameter() {defaultValue=897, group=Gravel, parameterName=GravelC_WorldScale}
  ScalarParameter() {defaultValue=592, group=Gravel, parameterName=GravelB_WorldScale}
  ScalarParameter() {defaultValue=0.5, group=Gravel, parameterName=Gravel_AtoB_BlendContrast}
  VectorParameter() {group=Gravel, parameterName=GravelC_ColorVariation, defaultValue=<rgba 1.3 0.9056 0.52 1>}
  VectorParameter() {group=Gravel, parameterName=GravelEdge_Color, defaultValue=<rgba 0.6 0.56 0.48 1>}
 ]}
