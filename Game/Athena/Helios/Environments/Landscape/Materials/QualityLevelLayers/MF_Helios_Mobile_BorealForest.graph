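# Mobile landscape layer material function for Helios boreal forest ground, outputting material attributes and a MobileLayerMask.
# One packed forest texture, sampled top-down in periodic world space, supplies normal and height data through named reroutes.
# Base color blends three color parameters by channel thresholds with alpha-based shading; roughness and specular reuse Mobile Gravel parameters.
# Layer weight and variation masks enter as inputs, and the layer mask is built from multiplied channels plus an offset.
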MaterialFunction /Game/Athena/Helios/Environments/Landscape/Materials/QualityLevelLayers/MF_Helios_Mobile_BorealForest
graph {nodes [
  n1 = NamedRerouteUsage()
  n2 = ComponentMask(None=@n1) {b=true}
  n3 = ComponentMask(None=@n1) {g=true}
  n4 = Saturate(None=Multiply(A=Multiply(A=@n2, B=@n3)) {constB=3})
  n5 = NamedRerouteUsage()
  FunctionOutput(None=Add(A=Multiply(A=@n4, B=@n5)) {constB=0.1}) {outputName=MobileLayerMask, sortPriority=1}
  n6 = SetMaterialAttributes!(
   MaterialAttributes
   BaseColor=NamedRerouteUsage()
   Normal=NamedRerouteUsage()
   Roughness=NamedRerouteUsage()
   Specular=NamedRerouteUsage()
   OpacityMask=NamedRerouteUsage()
  ) {attributeSetTypes [
    #1
    #2
    #3
    #4
    #5
   ]}
  FunctionOutput(None=@n6)
  n7 = Constant() {r=1024}
  n8 = TransformPosition!(None=WorldPosition() {worldPositionShaderOffset=WPT_CameraRelative}.XYZ, Periodic World Tile Size=@n7) {transformSourceType=TRANSFORMPOSSOURCE_TranslatedWorld, transformType=TRANSFORMPOSSOURCE_PeriodicWorld}
  n9 = TextureSample(UVs=ComponentMask(None=Divide(A=@n8, B=@n7)) {g=true, r=true}) {samplerType=SAMPLERTYPE_Masks, texture=/Game/Environments/Helios/Terrain/TexturesMobile/T_Helios_Mobile_Forest}
  NamedRerouteDeclaration(None=@n9.RGB) {name=Normals, nodeColor=<rgba 1 0.6094 0 1>}
  n10 = NamedRerouteUsage()
  NamedRerouteDeclaration(None=Multiply(A=ComponentMask(None=@n10) {b=true}, B=ScalarParameter() {defaultValue=0.3, group=Mobile Gravel, parameterName=Mobile Gravel Roughness})) {name=Specular, nodeColor=<rgba 0.4531 1 0 1>}
  NamedRerouteDeclaration(None=@n9.RGBA) {name=Texture, nodeColor=<rgba 0 1 0.2031 1>}
  n11 = ComponentMask(None=NamedRerouteUsage()) {a=true}
  n12 = ScalarParameter() {defaultValue=0.5, group=Mobile Gravel, parameterName=Mobile Gravel Pebble Roughness}
  n13 = OneMinus(None=Saturate(None=Multiply(A=@n11, B=@n12)))
  NamedRerouteDeclaration(None=Multiply(A=@n13, B=@n13)) {name=Roughness, nodeColor=<rgba 0.0312 1 0 1>}
  NamedRerouteDeclaration(None=FunctionInput() {inputName=LayerWeight, inputType=FunctionInput_Scalar}) {name=Layer Weight, nodeColor=<rgba 0 1 0.4844 1>}
  NamedRerouteDeclaration(None=ComponentMask(None=NamedRerouteUsage()) {b=true, g=true, r=true}) {name=Normal, nodeColor=<rgba 1 0 0.7969 1>}
  NamedRerouteDeclaration(None=ComponentMask(None=NamedRerouteUsage()) {a=true}) {name=Height, nodeColor=<rgba 0 1 0.3438 1>}
  n14 = VectorParameter() {group=Mobile Boreal Forest, parameterName=Mobile Boreal Forest Color 1, defaultValue=<rgba 0.0521 0.0521 0.0426 1>}
  n15 = VectorParameter() {group=Mobile Boreal Forest, parameterName=Mobile Boreal Forest Color 2, defaultValue=<rgba 0.082 0.0885 0.0189 1>}
  n16 = NamedRerouteUsage()
  n17 = ComponentMask(None=@n16) {b=true}
  n18 = Subtract(A=Multiply(A=@n17) {constB=5}) {constB=1.2}
  n19 = Multiply(A=@n17) {constB=6}
  n20 = LinearInterpolate(
   A=LinearInterpolate(A=@n14.RGB, B=@n15.RGB, Alpha=Saturate(None=@n18))
   B=VectorParameter() {group=Mobile Forest, parameterName=Mobile Forest Color 2, defaultValue=<rgba 0.3125 0.1162 0.066 1>}.RGB
   Alpha=Saturate(None=Subtract(A=@n19) {constB=2.3})
  )
  n21 = Add(A=ComponentMask(None=@n16) {a=true}) {constB=0.12}
  NamedRerouteDeclaration(None=Multiply(A=Multiply(A=@n20, B=@n21)) {constB=1.5}) {name=Base Color, nodeColor=<rgba 0.0547 0 1 1>}
  NamedRerouteDeclaration(None=FunctionInput() {inputName=VariationMasks Mobile, inputType=FunctionInput_Vector4, sortPriority=1}) {name=VariationMasks Mobile, nodeColor=<rgba 1 0.4596 0 1>}
  Reroute()
 ]}
