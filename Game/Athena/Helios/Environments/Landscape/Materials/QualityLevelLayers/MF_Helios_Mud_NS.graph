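# Material function that applies a mobile mud look to landscape material attributes in the Helios environment.
# It inverts the product of the incoming OpacityMask and a blue-channel mask, apparently from the mobile variation masks, then squares it and tints it with the Mobile Mud Color parameter.
# That result multiplies BaseColor, while Roughness is slightly reduced and OpacityMask is rewritten as its squared inverse.
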MaterialFunction /Game/Athena/Helios/Environments/Landscape/Materials/QualityLevelLayers/MF_Helios_Mud_NS
graph {nodes [
  n1 = GetMaterialAttributes(None=NamedRerouteUsage()) {attributeGetTypes [
    #1
    #2
    #3
    #4
   ]}
  n2 = ComponentMask(None=NamedRerouteUsage()) {b=true}
  n3 = Saturate(None=OneMinus(None=Multiply(A=@n1.OpacityMask, B=@n2)))
  n4 = Multiply(
   A=Multiply(A=@n3, B=@n3) {constB=2}
   B=VectorParameter() {group=Mobile Mud, parameterName=Mobile Mud Color, defaultValue=<rgba 0.7913 0.6514 0.5906 1>}.RGB
  )
  n5 = OneMinus(None=@n1.OpacityMask)
  n6 = Multiply(A=@n1.Roughness) {constB=0.865}
  FunctionOutput(None=SetMaterialAttributes!(
    MaterialAttributes=@n1.MaterialAttributes
    BaseColor=Multiply(A=@n1.BaseColor, B=@n4)
    OpacityMask=Multiply(A=@n5, B=@n5)
    Roughness=Saturate(None=@n6)
   ) {attributeSetTypes [#1, #2, #3]})
  NamedRerouteDeclaration(None=FunctionInput() {bUsePreviewValueAsDefault=true, inputName=VariationMasks Mobile, inputType=FunctionInput_Vector4, sortPriority=3}) {name=VariationMasks Mobile, nodeColor=<rgba 0.9219 1 0 1>}
  NamedRerouteDeclaration(None=FunctionInput() {bUsePreviewValueAsDefault=true, inputName=MobileInputMaterial, inputType=FunctionInput_MaterialAttributes, sortPriority=1}) {name=Input Material, nodeColor=<rgba 0 1 0.1094 1>}
 ]}
