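# Landscape layer material function for Helios salt flats, built from mobile textures and parameters as a lightweight quality-level variant.
# One packed mask texture, tiled in periodic world space from camera-relative position, supplies normal (RGB) and height (blue).
# Base color lerps two tint parameters by masks and applies derived darkening; roughness lerps between parameters by a squared mask.
# The output sets material attributes, with a squared, boosted, saturated mask in Opacity Mask.
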MaterialFunction /Game/Athena/Helios/Environments/Landscape/Materials/QualityLevelLayers/MF_Helios_SaltFflats_NS
graph {nodes [
  n1 = NamedRerouteUsage()
  n2 = Multiply(A=@n1, B=@n1)
  n3 = SetMaterialAttributes!(
   MaterialAttributes
   BaseColor=NamedRerouteUsage()
   Specular=NamedRerouteUsage()
   Roughness=NamedRerouteUsage()
   Normal=NamedRerouteUsage()
   OpacityMask=Saturate(None=Multiply(A=@n2) {constB=8})
  ) {
   desc=HighEnd Material
   attributeSetTypes [
    #1
    #2
    #3
    #4
    #5
   ]
  }
  FunctionOutput(None=@n3)
  n4 = Constant() {r=1800}
  n5 = TransformPosition!(None=WorldPosition() {worldPositionShaderOffset=WPT_CameraRelative}.XYZ, Periodic World Tile Size=@n4) {transformSourceType=TRANSFORMPOSSOURCE_TranslatedWorld, transformType=TRANSFORMPOSSOURCE_PeriodicWorld}
  n6 = TextureSample(UVs=ComponentMask(None=Divide(A=@n5, B=@n4)) {g=true, r=true}) {samplerType=SAMPLERTYPE_Masks, texture=/Game/Environments/Helios/Terrain/TexturesMobile/T_Helios_Mobile_SaltFlats}
  NamedRerouteDeclaration(None=@n6.B) {name=Height, nodeColor=<rgba 1 0 0.5156 1>}
  NamedRerouteDeclaration(None=FunctionInput() {inputName=VariationMasks Mobile, inputType=FunctionInput_Vector4}) {name=VariationMasks Mobile, nodeColor=<rgba 1 0.7773 0 1>}
  n7 = NamedRerouteUsage()
  NamedRerouteDeclaration(None=LinearInterpolate(
    A=ScalarParameter() {defaultValue=0.95, group=Mobile SaltFlats, parameterName=Mobile SaltFlats Roughness Min}
    B=ScalarParameter() {defaultValue=0.65, group=Mobile SaltFlats, parameterName=Mobile SaltFlats Roughness Max}
    Alpha=Multiply(A=@n7, B=@n7)
   )) {name=Roughness, nodeColor=<rgba 0 1 0.9531 1>}
  NamedRerouteDeclaration(None=ScalarParameter() {defaultValue=0.65, group=Mobile SaltFlats, parameterName=Mobile SaltFlats Specular}) {name=Specular, nodeColor=<rgba 0.0078 0 1 1>}
  NamedRerouteDeclaration(None=@n6.RGB) {name=Normal, nodeColor=<rgba 0.9219 1 0 1>}
  n8 = VectorParameter() {group=Mobile SaltFlats, parameterName=Mobile SaltFlats Color 1, defaultValue=<rgba 0.2969 0.2474 0.2474 1>}
  n9 = VectorParameter() {group=Mobile SaltFlats, parameterName=Mobile SaltFlats Color 2, defaultValue=<rgba 0.4062 0.4688 0.5 1>}
  n10 = NamedRerouteUsage()
  n11 = Subtract(A=Multiply(A=ComponentMask(None=@n10) {b=true}) {constB=2}) {constB=0.15}
  n12 = Multiply(A=Multiply(A=NamedRerouteUsage(), B=Saturate(None=@n11))) {constB=1.5}
  n13 = Abs(None=Add(A=NamedRerouteUsage()) {constB=-0.5})
  n14 = Multiply(A=Add(A=ComponentMask(None=@n13) {r=true}, B=ComponentMask(None=@n13) {r=true})) {constB=0.5}
  NamedRerouteDeclaration(None=Multiply(A=LinearInterpolate(A=@n8.RGB, B=@n9.RGB, Alpha=@n12), B=OneMinus(None=@n14))) {name=Base Color, nodeColor=<rgba 0.8984 1 0 1>}
 ]}
