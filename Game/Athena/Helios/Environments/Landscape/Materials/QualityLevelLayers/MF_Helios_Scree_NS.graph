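# Landscape material layer function for Helios scree, using a mobile packed texture and "Mobile Scree" parameters.
# One world-XY projected sample supplies normal in RGB, a texture mask in B and height in A, with roughness and specular computed through named reroutes.
# Base color blends dirt and two rock tints through derived masks, then applies a darkening term.
# MF_Scree_BiomeColorTint tints the attributes using four terrain mask inputs.
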MaterialFunction /Game/Athena/Helios/Environments/Landscape/Materials/QualityLevelLayers/MF_Helios_Scree_NS
graph {nodes [
  n1 = SetMaterialAttributes!(
   MaterialAttributes
   BaseColor=NamedRerouteUsage()
   Specular=NamedRerouteUsage()
   Roughness=NamedRerouteUsage()
   Normal=NamedRerouteUsage()
   OpacityMask=NamedRerouteUsage()
  ) {
   desc=HighEnd Material
   attributeSetTypes [
    #1
    #2
    #3
    #4
    #5
   ]
  }
  n2 = Constant()
  n3 = Constant()
  n4 = Constant()
  n5 = Constant()
  FunctionOutput(None=MaterialFunctionCall!(
    Material=@n1
    TerrainMasks_QuadA=FunctionInput(Preview=@n2) {bUsePreviewValueAsDefault=true, inputName=TerrainMasks_QuadA, inputType=FunctionInput_Vector4, sortPriority=10}
    TerrainMasks_QuadB=FunctionInput(Preview=@n3) {bUsePreviewValueAsDefault=true, inputName=TerrainMasks_QuadB, inputType=FunctionInput_Vector4, sortPriority=20}
    TerrainMasks_QuadC=FunctionInput(Preview=@n4) {bUsePreviewValueAsDefault=true, inputName=TerrainMasks_QuadC, inputType=FunctionInput_Vector4, sortPriority=30}
    TerrainMasks_QuadD=FunctionInput(Preview=@n5) {bUsePreviewValueAsDefault=true, inputName=TerrainMasks_QuadD, inputType=FunctionInput_Vector4, sortPriority=40}
   ) {materialFunction=/Game/Athena/Helios/Environments/Landscape/Materials/MaterialFunctions_Utilities/MF_Scree_BiomeColorTint}.Result)
  n6 = WorldPosition()
  n7 = TextureSample(UVs=Divide(A=TruncateLWC!(None=@n6.XY), B=ScalarParameter() {defaultValue=900, group=Mobile Scree, parameterName=Mobile Scree Scale})) {samplerType=SAMPLERTYPE_Masks, texture=/Game/Environments/Helios/Terrain/TexturesMobile/T_Helios_Mobile_Scree}
  NamedRerouteDeclaration(None=@n7.A) {name=Height, nodeColor=<rgba 1 0 0.5156 1>}
  n8 = NamedRerouteUsage()
  NamedRerouteDeclaration(None=Multiply(A=OneMinus(None=@n8), B=ScalarParameter() {defaultValue=0.6, group=Mobile Scree, parameterName=Mobile Scree Roughness})) {name=Roughness, nodeColor=<rgba 0 1 0.9531 1>}
  NamedRerouteDeclaration(None=Multiply(A=@n8, B=ScalarParameter() {defaultValue=1, group=Mobile Scree, parameterName=Mobile Scree Specular})) {name=Specular, nodeColor=<rgba 0.0078 0 1 1>}
  NamedRerouteDeclaration(None=@n7.RGB) {name=Normal, nodeColor=<rgba 0.9219 1 0 1>}
  n9 = VectorParameter() {group=Mobile Scree, parameterName=Mobile Scree Dirt, defaultValue=<rgba 0.1979 0.164 0.1185 1>}
  n10 = NamedRerouteUsage()
  n11 = LinearInterpolate(
   A=VectorParameter() {group=Mobile Scree, parameterName=Mobile Scree Rock1, defaultValue=<rgba 0.0721 0.0729 0.0698 1>}.RGB
   B=VectorParameter() {group=Mobile Scree, parameterName=Mobile Scree Rock 2, defaultValue=<rgba 0.8333 0.8148 0.7248 1>}.RGB
   Alpha=NamedRerouteDeclaration(None=Multiply(A=@n10, B=@n10) {constB=1.5}) {name=Rock Mask, nodeColor=<rgba 0 1 0.5781 1>}
  )
  n12 = Multiply(A=@n10) {constB=9}
  n13 = NamedRerouteDeclaration(None=Saturate(None=Subtract(A=@n12) {constB=2})) {name=Dirt Mask, nodeColor=<rgba 0.2656 1 0 1>}
  n14 = Abs(None=Add(A=NamedRerouteUsage()) {constB=-0.5})
  n15 = Multiply(A=Add(A=ComponentMask(None=@n14) {r=true}, B=ComponentMask(None=@n14) {r=true})) {constB=0.4}
  NamedRerouteDeclaration(None=Multiply(A=LinearInterpolate(A=@n9.RGB, B=@n11, Alpha=@n13), B=OneMinus(None=@n15))) {name=Base Color, nodeColor=<rgba 0.8984 1 0 1>}
  NamedRerouteDeclaration(None=@n7.B) {name=Texture Mask, nodeColor=<rgba 0.9688 0 1 1>}
 ]}
