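# Material function that blends the baked normal into the sedimentary cliff material for Helios landscape environments.
# It takes Baked Material and Sedimentary Material attribute inputs and splits both normals into components.
# The normals combine in whiteout style, adding R and G and multiplying B, with no normalize step.
# The output keeps the sedimentary attributes and replaces only Normal, so the baked material contributes nothing else.
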MaterialFunction /Game/Athena/Helios/Environments/Landscape/Materials/QualityLevelLayers/MF_Helios_CliffSedimentary_Blend_NS
graph {nodes [
  n1 = GetMaterialAttributes(None=FunctionInput() {inputName=Sedimentary Material, inputType=FunctionInput_MaterialAttributes, sortPriority=1}) {attributeGetTypes [#1]}
  n2 = MaterialFunctionCall(Float3=@n1.Normal) {materialFunction=&BreakOutFloat3Components}
  n3 = MaterialFunctionCall(Float3=GetMaterialAttributes(None=FunctionInput() {inputName=Baked Material, inputType=FunctionInput_MaterialAttributes}) {attributeGetTypes [#1]}.Normal) {materialFunction=&BreakOutFloat3Components}
  n4 = Add(A=@n2.R, B=@n3.R)
  n5 = Add(A=@n2.G, B=@n3.G)
  n6 = Multiply(A=@n2.B, B=@n3.B)
  FunctionOutput(None=SetMaterialAttributes!(
    MaterialAttributes=@n1.MaterialAttributes
    Normal=MaterialFunctionCall(R=@n4, G=@n5, B=@n6) {materialFunction=&AppendMany}.RGB
   ) {attributeSetTypes [#1]})
 ]}
