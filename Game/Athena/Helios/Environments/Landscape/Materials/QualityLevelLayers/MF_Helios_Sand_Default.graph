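# Helios landscape material function for the default sand layer, outputting a main Material and a darkened UnderOceanSandMaterial.
# World-space sand textures feed color, specular, roughness, normals and displacement, with quality-switched normal strength.
# World height against ocean-level offsets drives height-lerp masks for wet and underwater shading, plus an optional DesertSaltLake wet tint.
# A UseOceanLevel static switch picks the attribute set, and MF_Sand_BiomeColorTint tints both outputs using terrain masks.
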MaterialFunction /Game/Athena/Helios/Environments/Landscape/Materials/QualityLevelLayers/MF_Helios_Sand_Default
graph {
 properties {description=-3909}
 nodes [
  n1 = ScalarParameter() {
   defaultValue=55
   desc="This offsets the ocean level used by the sand. Positive numbers go lower, negative numbers make it go higher."
   group=Sand
   parameterName=Sand_BaseOceanLevelOffset
  }
  n2 = Add(A=NamedRerouteDeclaration(None=@n1) {name=WaterLevelOffset, nodeColor=<rgba 1 0 1 1>}, B=ComponentMask(None=NamedRerouteUsage()) {b=true})
  Abs(None=@n2)
  n3 = Multiply(A=NamedRerouteUsage(), B=ScalarParameter() {defaultValue=3, group=Sand, parameterName=Sand_WL_HeightMod})
  Add(A=@n3, B=ScalarParameter() {group=Sand, parameterName=Sand_WL_HeightAdd})
  Constant() {r=55}
  Constant() {r=600}
  n4 = SetMaterialAttributes!(
   MaterialAttributes
   BaseColor=Multiply(A=ScalarParameter() {defaultValue=0.38, parameterName=UnderOceanDarkness}, B=NamedRerouteUsage())
   Specular=ScalarParameter() {defaultValue=0.5, parameterName=UnderOceanSpec}
   Roughness=ScalarParameter() {defaultValue=0.9, parameterName=UnderOceanRough}
   Normal=VertexNormalWS()
  ) {attributeSetTypes [
    #1
    #2
    #3
    #4
   ]}
  FunctionOutput(None=MaterialFunctionCall!(
    Material=@n4
    TerrainMasks_QuadA=NamedRerouteUsage()
    TerrainMasks_QuadB=NamedRerouteUsage()
    TerrainMasks_QuadC=NamedRerouteUsage()
    TerrainMasks_QuadD=NamedRerouteUsage()
    OD_MD_SedimentLayers
   ) {materialFunction=/Game/Athena/Helios/Environments/Landscape/Materials/MaterialFunctions_Utilities/MF_Sand_BiomeColorTint}.Result) {outputName=UnderOceanSandMaterial}
  n5 = NamedRerouteUsage()
  n6 = Saturate(None=Multiply(A=VectorParameter() {group=DesertSaltLake, parameterName=DesertSaltLake_ColorTintWetSand, defaultValue=<rgba 0.5 0.5 0.5 0>}.RGB, B=@n5))
  n7 = Add(A=WorldPosition().Z, B=ScalarParameter() {defaultValue=-1450, group=DesertSaltLake, parameterName=DesertSaltLake_WetHeight, sortPriority=0})
  n8 = ScalarParameter() {defaultValue=150, group=DesertSaltLake, parameterName=DesertSaltLake_WetFalloff, sortPriority=0}
  n9 = NamedRerouteUsage()
  n10 = TextureSample(UVs=@n9) {samplerSource=SSM_Wrap_WorldGroupSettings, texture=/Game/Environments/Helios/Terrain/Textures/T_Helios_Sand_01_D}
  n11 = Power(Base=NamedRerouteDeclaration(None=@n10.A) {name=TextureSandHeight, nodeColor=<rgba 1 0 1 1>}, Exp=Constant() {r=0.2})
  n12 = NamedRerouteDeclaration(None=OneMinus(None=NamedRerouteDeclaration(None=@n11) {name=ModifiedSandHeight, nodeColor=<rgba 1 0 1 1>})) {name=InvertedSandHeight, nodeColor=<rgba 1 0 1 1>}
  n13 = MaterialFunctionCall(
   Transition Phase=Saturate(None=Divide(A=@n7, B=@n8))
   Height Texture=@n12
   Contrast=ScalarParameter() {defaultValue=0.5, group=DesertSaltLake, parameterName=DesertSaltLake_WetContrast, sortPriority=0}
  ) {materialFunction=&HeightLerp}
  n14 = StaticBoolParameter() {desc=Only select one option per layer as true, group=LayerUsage, parameterName=UseAsphalt_DesertSaltLake, sortPriority=-500}
  n15 = TextureSample(UVs=@n9) {texture=/Game/Environments/Helios/Terrain/Textures/T_Helios_Sand_01_Displacement}
  n16 = SetMaterialAttributes!(
   MaterialAttributes
   BaseColor=StaticSwitch(True=LinearInterpolate(A=@n6, B=@n5, Alpha=@n13.Alpha), False=@n5, Value=@n14)
   Specular=NamedRerouteUsage()
   Roughness=NamedRerouteUsage()
   Normal=NamedRerouteUsage()
   OpacityMask=NamedRerouteUsage()
   Displacement=LinearInterpolate(B=@n15.R, Alpha=ScalarParameter() {defaultValue=0.65, group=Displacement, parameterName=Sand_DisplacementAmt}) {constA=0.5}
  ) {attributeSetTypes [
    #1
    #2
    #3
    #4
    #5
    #6
   ]}
  n17 = NamedRerouteUsage()
  n18 = Add(A=Multiply(A=@n17, B=Constant() {r=4}), B=Constant() {r=0.5})
  n19 = Add(A=Multiply(A=@n17, B=Constant() {r=3}), B=Constant() {r=0.65})
  n20 = SetMaterialAttributes!(
   MaterialAttributes
   BaseColor=NamedRerouteUsage()
   Specular=Multiply(A=Saturate(None=@n18), B=NamedRerouteUsage())
   Roughness=Multiply(A=Saturate(None=@n19), B=NamedRerouteUsage())
   Normal=NamedRerouteUsage()
   OpacityMask=NamedRerouteUsage()
  ) {attributeSetTypes [
    #1
    #2
    #3
    #4
    #5
   ]}
  n21 = FunctionInput(Preview=StaticBool() {value=true}) {bUsePreviewValueAsDefault=true, description="Default=True", inputName=UseOceanLevel, inputType=FunctionInput_StaticBool, sortPriority=200}
  n22 = FunctionInput(Preview=Constant()) {bUsePreviewValueAsDefault=true, inputName=TerrainMasks_QuadA, inputType=FunctionInput_Vector4, sortPriority=-400}
  n23 = FunctionInput(Preview=Constant()) {bUsePreviewValueAsDefault=true, inputName=TerrainMasks_QuadB, inputType=FunctionInput_Vector4, sortPriority=-300}
  n24 = FunctionInput(Preview=Constant()) {bUsePreviewValueAsDefault=true, inputName=TerrainMasks_QuadC, inputType=FunctionInput_Vector4, sortPriority=-200}
  n25 = FunctionInput(Preview=Constant()) {bUsePreviewValueAsDefault=true, inputName=TerrainMasks_QuadD, inputType=FunctionInput_Vector4, sortPriority=-100}
  FunctionOutput(None=MaterialFunctionCall(
    Material=StaticSwitch(True=@n16, False=@n20, Value=@n21)
    TerrainMasks_QuadA=NamedRerouteDeclaration(None=@n22) {name=TerrainMasks_QuadA, nodeColor=<rgba 1 0 1 1>}
    TerrainMasks_QuadB=NamedRerouteDeclaration(None=@n23) {name=TerrainMasks_QuadB, nodeColor=<rgba 1 0 1 1>}
    TerrainMasks_QuadC=NamedRerouteDeclaration(None=@n24) {name=TerrainMasks_QuadC, nodeColor=<rgba 1 0 1 1>}
    TerrainMasks_QuadD=NamedRerouteDeclaration(None=@n25) {name=TerrainMasks_QuadD, nodeColor=<rgba 1 0 1 1>}
    OD_MD_SedimentLayers=NamedRerouteUsage()
   ) {materialFunction=&MF_Sand_BiomeColorTint}.Result) {outputName=Material}
  n26 = ScalarParameter() {defaultValue=1.5, group=Sand, parameterName=Sand_Water_RoughMod}
  n27 = NamedRerouteUsage()
  n28 = ScalarParameter() {defaultValue=0.5, group=Sand, parameterName=Sand_Wet_RoughMod}
  n29 = NamedRerouteUsage()
  n30 = LinearInterpolate(
   A=ScalarParameter() {defaultValue=0.55, group=Sand, parameterName=Sand_Wet_RoughFLAT}
   B=Saturate(None=Multiply(A=@n28, B=@n29))
   Alpha=ScalarParameter() {defaultValue=0.5, group=Sand, parameterName=Sand_Wet_RoughBlend}
  )
  n31 = Multiply(A=NamedRerouteUsage(), B=ScalarParameter() {defaultValue=2, group=Sand, parameterName=Sand_Underwater_RoughMASKMod})
  n32 = ScalarParameter() {group=Sand, parameterName=Sand_Underwater_RoughMASKAdd}
  LinearInterpolate(
   A=Saturate(None=Multiply(A=@n26, B=@n27))
   B=LinearInterpolate(A=Saturate(None=@n30), B=@n29, Alpha=NamedRerouteUsage())
   Alpha=Saturate(None=Add(A=@n31, B=@n32))
  )
  n33 = NamedRerouteUsage()
  n34 = Multiply(
   A=Max(A=NamedRerouteUsage(), B=OneMinus(None=@n33))
   B=ScalarParameter() {defaultValue=2.5, group=Sand, parameterName=Sand_Wet_NormalMaskMod}
  )
  n35 = ScalarParameter() {defaultValue=-1.5, group=Sand, parameterName=Sand_Wet_NormalMaskAdd}
  LinearInterpolate(A=NamedRerouteUsage(), Alpha=Saturate(None=Add(A=@n34, B=@n35)))
  LinearInterpolate(
   A=ScalarParameter() {defaultValue=4, group=Sand, parameterName=Sand_Variance_HeightPowA}
   B=ScalarParameter() {defaultValue=1, group=Sand, parameterName=Sand_Variance_HeightPowB}
   Alpha=NamedRerouteUsage()
  )
  LinearInterpolate(
   A=Multiply(A=NamedRerouteUsage(), B=NamedRerouteUsage())
   B=Multiply(A=NamedRerouteUsage(), B=NamedRerouteUsage())
   Alpha=MaterialFunctionCall(
    UVs=MaterialFunctionCall!(WorldPosition) {materialFunction=/Game/Athena/Environments/Landscape/MaterialFunctions/Utilities/MF_Apollo_Terrain_Coords}.Adjustable
    Texture Scale=Constant() {r=1}
   ) {materialFunction=&ScaleUVsByCenter}.0-1 mask
  )
  n36 = NamedRerouteUsage()
  n37 = Saturate(None=Add(A=NamedRerouteUsage(), B=ScalarParameter() {defaultValue=0.8, group=Sand, parameterName=Sand_Wet_1_Dark}))
  n38 = Multiply(A=NamedRerouteUsage(), B=ScalarParameter() {defaultValue=0.75, group=Sand, parameterName=Sand_Wet2_Mask_DiffuseMod})
  n39 = ScalarParameter() {defaultValue=-0.4, group=Sand, parameterName=Sand_Wet2_Mask_DiffuseAdd}
  n40 = Add(A=Saturate(None=Add(A=@n38, B=@n39)), B=ScalarParameter() {defaultValue=0.8, group=Sand, parameterName=Sand_Wet_2_Dark})
  n41 = NamedRerouteUsage()
  n42 = ScalarParameter() {defaultValue=0.8, group=Sand, parameterName=Sand_Water_Darkness}
  Multiply(A=Multiply(A=Multiply(A=@n36, B=@n37), B=Saturate(None=@n40)), B=Saturate(None=Add(A=@n41, B=@n42)))
  n43 = NamedRerouteUsage()
  Multiply(A=@n43, B=Constant() {r=1.5})
  n44 = Divide(
   A=Add(A=@n2, B=ScalarParameter() {defaultValue=5, group=Sand, parameterName=Sand_Wet_Offset})
   B=ScalarParameter() {defaultValue=100, group=Sand, parameterName=Sand_Wet_Distance}
  )
  n45 = Power(Base=NamedRerouteUsage(), Exp=ScalarParameter() {defaultValue=0.2, group=Sand, parameterName=Sand_TextureHeightPower11A})
  n46 = Add(
   A=Multiply(A=@n45, B=ScalarParameter() {defaultValue=0.5, group=Sand, parameterName=Sand_Wet_HeightMod})
   B=ScalarParameter() {defaultValue=0.1, group=Sand, parameterName=Sand_Wet_HeightAdd}
  )
  NamedRerouteDeclaration(None=MaterialFunctionCall(Transition Phase=Saturate(None=@n44), Height Texture=Saturate(None=@n46)) {materialFunction=&HeightLerp}.Lerp Alpha No Contrast) {name=WetSandMask, desc="0=Wet", nodeColor=<rgba 1 0 1 1>}
  n47 = NamedRerouteUsage()
  n48 = TextureSample(UVs=Divide(A=ComponentMask(None=@n47) {g=true, r=true}, B=Constant() {r=85}) {constB=8192}) {samplerSource=SSM_Wrap_WorldGroupSettings, samplerType=SAMPLERTYPE_LinearColor, texture=/Game/Global/Textures/Default/Noise/T_Apollo_TCNoise}
  n49 = NamedRerouteUsage()
  n50 = Add(A=Subtract(A=@n48.RGB, B=Constant() {r=0.5}), B=@n49)
  n51 = Saturate(None=LinearInterpolate(A=Constant() {r=-5}, B=Constant() {r=1.5}, Alpha=NamedRerouteUsage()))
  n52 = LinearInterpolate(A=Constant() {r=-3}, B=Constant() {r=1}, Alpha=NamedRerouteUsage())
  n53 = Normalize(VectorInput=LinearInterpolate(A=NamedRerouteUsage(), B=LinearInterpolate(A=@n50, B=@n49, Alpha=@n51), Alpha=Saturate(None=@n52) {desc=Less on slopes}))
  n54 = Constant3Vector() {constant=<rgba 0.5 0.5 0 0>}
  n55 = TextureSample(UVs=NamedRerouteUsage()) {
   samplerSource=SSM_Wrap_WorldGroupSettings
   samplerType=SAMPLERTYPE_Normal
   texture=/Game/Environments/Helios/Terrain/Textures/T_Helios_Sand_01_N
  }
  NamedRerouteDeclaration(None=Add(A=@n55.RGB, B=Multiply(A=@n53, B=@n54))) {name=EpicNormals, nodeColor=<rgba 1 0 1 1>}
  NamedRerouteDeclaration(None=MaterialFunctionCall!(Coordinates, R_01_Scale, R_02_Scale, R_03_Scale, R_04_Scale, G_01_Scale=Constant() {r=8192}, G_02_Scale, G_03_Scale, G_04_Scale) {materialFunction=/Game/Athena/Helios/Environments/Landscape/Materials/MaterialFunctions_Utilities/MF_VariationMasks2}.G_01) {name=SandVariance, nodeColor=<rgba 1 0 1 1>}
  n56 = VertexNormalWS()
  NamedRerouteDeclaration(None=Abs(None=ComponentMask(None=@n56) {b=true})) {name=VertexNormalZMask, nodeColor=<rgba 1 0.2578 0 1>}
  n57 = TextureSample(UVs=@n43) {
   samplerSource=SSM_Wrap_WorldGroupSettings
   samplerType=SAMPLERTYPE_Masks
   texture=/Game/Environments/Helios/Terrain/Textures/T_Helios_Sand_01_S
  }
  n58 = NamedRerouteDeclaration(None=@n57.B) {name=RawRough, nodeColor=<rgba 1 0 1 1>}
  n59 = Multiply(A=Constant() {r=1.75}, B=@n58)
  n60 = Constant() {r=-1.5}
  n61 = NamedRerouteDeclaration(None=@n57.R) {name=RawSpec, nodeColor=<rgba 1 0 1 1>}
  n62 = Multiply(A=Saturate(None=Add(A=@n59, B=@n60)), B=@n61)
  n63 = Max(A=NamedRerouteUsage(), B=OneMinus(None=NamedRerouteUsage()))
  NamedRerouteDeclaration(None=LinearInterpolate(A=@n62, B=@n61, Alpha=@n63)) {name=Specular, nodeColor=<rgba 1 0 1 1>}
  n64 = FunctionInput(Preview=StaticBool()) {
   bUsePreviewValueAsDefault=true
   description="Default = False"
   inputName=UseTangentNormals
   inputType=FunctionInput_StaticBool
   sortPriority=15
  }
  NamedRerouteDeclaration(None=@n64) {name=UseTangentNormalsBool, nodeColor=<rgba 0.5 0 1 1>}
  n65 = ScalarParameter() {
   defaultValue=70
   desc="This offsets the ocean level used by the sand. Positive numbers go lower, negative numbers make it go higher."
   group=Sand
   parameterName=Sand_BaseOceanLevel
  }
  n66 = Add(A=@n65, B=ComponentMask(None=NamedRerouteUsage()) {b=true})
  n67 = Saturate(None=Divide(A=@n66, B=Constant() {r=25}))
  n68 = Add(A=Multiply(A=NamedRerouteUsage(), B=Constant() {r=4}), B=Constant() {r=-0.5})
  n69 = MaterialFunctionCall(Transition Phase=@n67, Height Texture=Saturate(None=@n68)) {materialFunction=&HeightLerp}
  n70 = OneMinus(None=NamedRerouteUsage())
  NamedRerouteDeclaration(None=LinearInterpolate(A=Constant() {r=0.3}, B=@n58, Alpha=Max(A=@n69.Lerp Alpha No Contrast, B=@n70))) {name=Rough, nodeColor=<rgba 1 0 1 1>}
  n71 = NamedRerouteUsage()
  n72 = NamedRerouteUsage()
  n73 = Constant() {r=2}
  n74 = Saturate(None=Add(A=Multiply(A=@n72, B=@n73), B=Constant() {r=-1}))
  n75 = MaterialFunctionCall!(
   Normals=LinearInterpolate(A=NamedRerouteUsage(), B=NamedRerouteUsage(), Alpha=Multiply(A=@n71, B=@n74))
   TransformToTangentSpace=NamedRerouteUsage()
   Normal Strength=QualitySwitch(
    Default=ScalarParameter() {defaultValue=1, group=Sand, parameterName=Sand_NormalStrength}
    Epic=ScalarParameter() {defaultValue=1.25, group=Sand, parameterName=Sand_NormalStrength_Epic}
   )
  ) {materialFunction=/Game/Athena/Apollo/Environments/Landscape/Materials/Biome/CommonMaterials/MF_HQ_WorldAlignedNormal}
  n76 = Constant3Vector()
  n77 = Add(
   A=@n75.Result
   B=NamedRerouteDeclaration(None=FunctionInput(Preview=@n76) {bUsePreviewValueAsDefault=true, inputName=OD_MD_Normals, sortPriority=50}) {name=OD_MD_Normals, nodeColor=<rgba 0.5 0 1 1>}
  )
  NamedRerouteDeclaration(None=StaticSwitch(True=Normalize(VectorInput=@n77), False=@n75.Result, Value=NamedRerouteUsage())) {name=FinalNormal, nodeColor=<rgba 1 0 1 1>}
  NamedRerouteDeclaration(None=Constant3Vector() {constant=<rgba 0 0 1 0>}) {name=001, nodeColor=<rgba 0 0 1 1>}
  NamedRerouteDeclaration(None=WorldPosition().XYZ) {name=WorldPos, nodeColor=<rgba 0.6172 1 0 1>}
  NamedRerouteDeclaration(None=@n55.RGB) {name=Pre-Normals, nodeColor=<rgba 1 0 1 1>}
  NamedRerouteDeclaration(None=MaterialFunctionCall(Normals=@n55.RGB, TransformToTangentSpace=NamedRerouteUsage()) {materialFunction=&MF_HQ_WorldAlignedNormal}.Result) {name=NormalsNoWater, nodeColor=<rgba 1 0 1 1>}
  n78 = Divide(A=@n2, B=ScalarParameter() {defaultValue=15, group=Sand, parameterName=Sand_Water_Distance})
  n79 = Add(
   A=Multiply(A=@n45, B=ScalarParameter() {defaultValue=2, group=Sand, parameterName=Sand_Water_HeightMod})
   B=ScalarParameter() {defaultValue=-0.75, group=Sand, parameterName=Sand_Water_HeightAdd}
  )
  NamedRerouteDeclaration(None=MaterialFunctionCall(Transition Phase=Saturate(None=@n78), Height Texture=Saturate(None=@n79)) {materialFunction=&HeightLerp}.Lerp Alpha No Contrast) {name=UnderwaterSandMask, desc="0=Underwater", nodeColor=<rgba 1 0 1 1>}
  n80 = Add(A=@n2, B=ScalarParameter() {defaultValue=5, group=Sand, parameterName=Sand_Wet_Rough_Offset})
  n81 = ScalarParameter() {defaultValue=35, group=Sand, parameterName=Sand_Wet_Rough_Distance}
  n82 = Multiply(A=@n45, B=ScalarParameter() {defaultValue=1, group=Sand, parameterName=Sand_Wet_Rough_HeightMod})
  n83 = ScalarParameter() {group=Sand, parameterName=Sand_Wet_Rough_HeightAdd}
  n84 = MaterialFunctionCall(Transition Phase=Saturate(None=Divide(A=@n80, B=@n81)), Height Texture=Saturate(None=Add(A=@n82, B=@n83))) {materialFunction=&HeightLerp}
  NamedRerouteDeclaration(None=@n84.Lerp Alpha No Contrast) {name=WetSandMask_Rough, desc="0=Wet", nodeColor=<rgba 1 0 1 1>}
  NamedRerouteDeclaration(None=@n84.Lerp Alpha No Contrast) {name=WetSandMask_Normal, desc=Not used, nodeColor=<rgba 1 0 1 1>}
  NamedRerouteDeclaration(None=MaterialFunctionCall(Transition Phase=@n67, Height Texture=@n68) {materialFunction=&HeightLerp}.Lerp Alpha No Contrast) {name=WaterLevel-HeightMask, nodeColor=<rgba 1 0 1 1>}
  NamedRerouteDeclaration(None=@n62) {name=UnderOcean-Spec, nodeColor=<rgba 1 0 1 1>}
  n85 = Divide(
   A=Add(A=@n66, B=ScalarParameter() {defaultValue=5, group=Sand, parameterName=Sand_UW_Offset})
   B=ScalarParameter() {defaultValue=15, group=Sand, parameterName=Sand_UW_Distance}
  )
  n86 = Add(
   A=Multiply(A=NamedRerouteUsage(), B=ScalarParameter() {defaultValue=2, group=Sand, parameterName=Sand_UW_HeightMod})
   B=ScalarParameter() {defaultValue=-0.75, group=Sand, parameterName=Sand_UW_HeightAdd}
  )
  NamedRerouteDeclaration(None=MaterialFunctionCall(Transition Phase=Saturate(None=@n85), Height Texture=Saturate(None=@n86)) {materialFunction=&HeightLerp}.Lerp Alpha No Contrast) {name=ShorelineNormalMask, nodeColor=<rgba 1 0 1 1>}
  NamedRerouteDeclaration(None=Saturate(None=@n3)) {name=UnderwaterMask, desc="0=Underwater", nodeColor=<rgba 1 0 1 1>}
  NamedRerouteDeclaration(None=FunctionInput(Preview=Constant4Vector().RGBA) {bUsePreviewValueAsDefault=true, inputName=OD_MD_SedimentLayers, inputType=FunctionInput_Vector4, sortPriority=60}) {name=OD_MD_SedimentLayers, nodeColor=<rgba 0.5 0 1 1>}
  NamedRerouteDeclaration(None=StaticBoolParameter() {group=OpenDesert_MacroDetail, parameterName=OD_MD_Use_OpenDesert_MacroDetail, sortPriority=-9999}) {name=OD_MD_Bool, nodeColor=<rgba 0.5 0 1 1>}
  n87 = Add(A=@n66, B=Constant() {r=10})
  n88 = Constant() {r=125}
  NamedRerouteDeclaration(None=Saturate(None=Divide(A=@n87, B=@n88))) {name=WaterLevel, nodeColor=<rgba 1 0 1 1>}
  n89 = Multiply(
   A=LinearInterpolate(A=Constant() {r=1}, B=Constant() {r=0.6}, Alpha=NamedRerouteUsage())
   B=NamedRerouteDeclaration(None=@n10.RGB) {name=BasecolorTexture, nodeColor=<rgba 1 0 1 1>}
  )
  n90 = Saturate(None=Add(A=ScalarParameter() {defaultValue=0.7, group=Sand, parameterName=Sand_Underwater_Darkness}, B=NamedRerouteUsage()))
  n91 = ScalarParameter() {defaultValue=0.6, group=Sand, parameterName=Sand_Wet_Darkness}
  n92 = MaterialFunctionCall!(Diffuse 1, Diffuse 2, Transistion Phase=NamedRerouteUsage(), Height Texture 1=@n12, Height Texture 2=NamedRerouteUsage(), Contrast) {materialFunction=/Engine/Functions/Engine_MaterialFunctions02/Texturing/HeightLerpWithTwoHeightMaps}
  n93 = Saturate(None=Add(A=NamedRerouteDeclaration(None=@n91) {name=UnderWaterSandDarkness, nodeColor=<rgba 1 0 1 1>}, B=OneMinus(None=@n92.Lerp Alpha No Contrast)))
  NamedRerouteDeclaration(None=Multiply(A=NamedRerouteDeclaration(None=@n89) {name=BaseColorNoWater, nodeColor=<rgba 1 0 1 1>}, B=Multiply(A=@n90, B=@n93))) {name=BaseColorWithWater, nodeColor=<rgba 1 0 1 1>}
  NamedRerouteDeclaration(None=@n92.Lerp Alpha No Contrast) {name=WaterLevel-HeightMask2, nodeColor=<rgba 1 0 1 1>}
  NamedRerouteDeclaration(None=MaterialFunctionCall!(Scale=ScalarParameter() {defaultValue=877, group=Grass_BorealForest, parameterName=Sand_WorldScale}) {materialFunction=/Game/Athena/Apollo/Environments/Landscape/Materials/Biome/CommonMaterials/MF_SimpleWorldUV}.Result) {name=SandUVs, nodeColor=<rgba 1 0 1 1>}
  NamedRerouteUsage()
  NamedRerouteUsage()
  QualitySwitch(Default=@n55.RGB, Epic=NamedRerouteUsage())
  n94 = NamedRerouteUsage()
  n95 = Add(A=NamedRerouteUsage(), B=Constant() {r=191})
  Saturate(None=Divide(A=Add(A=@n94, B=@n95), B=Constant() {r=200}))
  n96 = ScalarParameter() {defaultValue=1, group=Displacement, parameterName=Sand_Displacement_Power}
  Saturate(None=Multiply(A=Power(Base=@n15.R, Exp=@n96), B=ScalarParameter() {defaultValue=1, group=Displacement, parameterName=Sand_Displacement_Mod}))
  Saturate(None=LinearInterpolate(A=Constant() {r=-3}, B=Constant() {r=1}, Alpha=NamedRerouteUsage())) {desc=Less on slopes}
  n97 = Constant() {r=0.3}
  Saturate(None=Add(A=NamedRerouteUsage(), B=LinearInterpolate(A=@n97, B=@n58, Alpha=@n63)))
  ScalarParameter() {defaultValue=-5, group=Sand, parameterName=Sand_NoiseHeightA}
  ScalarParameter() {defaultValue=1.5, group=Sand, parameterName=Sand_NoiseHeightB}
  ScalarParameter() {defaultValue=10, group=Sand, parameterName=Sand_SWSHO}
  ScalarParameter() {defaultValue=125, group=Sand, parameterName=Sand_WLD}
 ]
}
