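# Material function for the Helios landscape sand layer at one quality level, built on a mobile sand texture and parameters.
# One packed texture, sampled by world XY, feeds procedurally varied sand color, specular and normals, with roughness from a scalar.
# A world-height waterline mask is computed, and normal strength is mask-driven, optionally modulated through the UseGrass_DesertGrass static switch.
# The resulting material attributes pass through MF_Sand_BiomeColorTint with terrain mask inputs before output.
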MaterialFunction /Game/Athena/Helios/Environments/Landscape/Materials/QualityLevelLayers/MF_Helios_Sand_NS
graph {nodes [
  ComponentMask() {g=true, r=true}
  n1 = SetMaterialAttributes!(
   MaterialAttributes
   BaseColor=NamedRerouteUsage()
   Specular=NamedRerouteUsage()
   Roughness=NamedRerouteUsage()
   Normal=NamedRerouteUsage()
   OpacityMask=NamedRerouteUsage()
  ) {
   desc=HighEnd Material
   attributeSetTypes [
    #1
    #2
    #3
    #4
    #5
   ]
  }
  FunctionOutput(None=MaterialFunctionCall!(
    Material=@n1
    TerrainMasks_QuadA=NamedRerouteUsage()
    TerrainMasks_QuadB=NamedRerouteUsage()
    TerrainMasks_QuadC=NamedRerouteUsage()
    TerrainMasks_QuadD=NamedRerouteUsage()
    OD_MD_SedimentLayers=NamedRerouteUsage()
   ) {materialFunction=/Game/Athena/Helios/Environments/Landscape/Materials/MaterialFunctions_Utilities/MF_Sand_BiomeColorTint}.Result)
  n2 = WorldPosition()
  n3 = TextureSample(UVs=Divide(A=TruncateLWC!(None=@n2.XY), B=ScalarParameter() {defaultValue=770, group=Mobile Sand, parameterName=Mobile Sand Scale, sortPriority=-1000})) {samplerType=SAMPLERTYPE_Masks, texture=/Game/Environments/Helios/Terrain/TexturesMobile/T_Helios_Mobile_Sand}
  NamedRerouteDeclaration(None=@n3.RGBA) {name=Texture, nodeColor=<rgba 1 0.8672 0 1>}
  NamedRerouteDeclaration(None=ScalarParameter() {defaultValue=0.6, group=Mobile Sand, parameterName=Mobile Sand Roughness}) {name=Roughness, nodeColor=<rgba 0.1484 0 1 1>}
  NamedRerouteDeclaration(None=FunctionInput(Preview=Constant()) {bUsePreviewValueAsDefault=true, inputName=TerrainMasks_QuadD, inputType=FunctionInput_Vector4, sortPriority=-100}) {name=TerrainMask_QuadD, nodeColor=<rgba 1 0 0.375 1>}
  n4 = NamedRerouteUsage()
  NamedRerouteDeclaration(None=Multiply(A=ComponentMask(None=@n4) {b=true}, B=ScalarParameter() {defaultValue=0.3, group=Mobile Sand, parameterName=Mobile Sand Specular}) {constB=2}) {name=Specular, nodeColor=<rgba 0.5 0 1 1>}
  n5 = ComponentMask(None=NamedRerouteUsage()) {b=true}
  n6 = NamedRerouteUsage()
  n7 = Frac(None=Multiply(A=ComponentMask(None=@n6) {b=true, g=true, r=true}) {constB=6})
  n8 = LinearInterpolate(
   A=Multiply(A=VectorParameter() {group=Mobile Sand, parameterName=Mobile Sand Color, defaultValue=<rgba 0.5104 0.3188 0.1834 1>}.RGB, B=@n5)
   B=LinearInterpolate(
    A=VectorParameter() {group=Mobile Sand, parameterName=Mobile Sand Color 3, defaultValue=<rgba 0.2917 0.2332 0.1838 1>}.RGB
    B=Step(Y=@n7) {constX=0.51}
    Alpha=ScalarParameter() {defaultValue=0.04, group=Mobile Sand, parameterName=Mobile Sand Color Variation}
   ) {constA=0.5}
   Alpha=Saturate(None=LinearInterpolate(Alpha=@n5) {constA=11, constB=-3})
  )
  NamedRerouteDeclaration(None=LinearInterpolate(A=Multiply(A=@n8) {constB=0.6}, B=@n8, Alpha=NamedRerouteUsage())) {name=BaseColor, nodeColor=<rgba 0.5 0 1 1>}
  NamedRerouteDeclaration(None=FunctionInput() {bUsePreviewValueAsDefault=true, inputName=VariationMasks Mobile, inputType=FunctionInput_Vector4, sortPriority=3}) {name=VariationMasks Mobile, nodeColor=<rgba 0.9219 1 0 1>}
  n9 = NamedRerouteUsage()
  n10 = NamedRerouteUsage()
  n11 = Add(A=Multiply(A=ComponentMask(None=@n10) {b=true}) {constB=2.4}) {constB=-0.3}
  n12 = NamedRerouteUsage()
  n13 = ComponentMask(None=@n12) {b=true}
  n14 = ComponentMask(None=@n12) {a=true}
  n15 = NamedRerouteUsage()
  n16 = Saturate(None=Add(A=Multiply(A=@n15) {constB=2}) {constB=-0.5})
  n17 = Multiply(A=LinearInterpolate(B=Saturate(None=@n11), Alpha=Max(A=@n13, B=@n14)) {constA=1}, B=@n16)
  n18 = StaticBoolParameter() {group=LayerUsage, parameterName=UseGrass_DesertGrass}
  NamedRerouteDeclaration(None=LinearInterpolate(B=ComponentMask(None=@n9) {b=true, g=true, r=true}, Alpha=StaticSwitch(True=@n17, False=@n16, Value=@n18)) {constA=0.5}) {name=Normals, nodeColor=<rgba 0.5 0 1 1>}
  NamedRerouteDeclaration(None=FunctionInput(Preview=Constant()) {bUsePreviewValueAsDefault=true, inputName=TerrainMasks_QuadA, inputType=FunctionInput_Vector4, sortPriority=-400}) {name=TerrainMask_QuadA, nodeColor=<rgba 1 0 0.375 1>}
  NamedRerouteDeclaration(None=FunctionInput(Preview=Constant()) {bUsePreviewValueAsDefault=true, inputName=TerrainMasks_QuadB, inputType=FunctionInput_Vector4, sortPriority=-300}) {name=TerrainMask_QuadB, nodeColor=<rgba 1 0 0.375 1>}
  n19 = Add(
   A=TruncateLWC(None=WorldPosition().Z)
   B=ScalarParameter() {defaultValue=40, group=Mobile Sand, parameterName=Mobile Sand WaterLine Position}
  ) {constB=70}
  n20 = ScalarParameter() {defaultValue=0.03, group=Mobile Sand, parameterName=Mobile Sand WaterLine Contrast}
  n21 = Multiply(A=Add(A=Multiply(A=@n19, B=@n20) {constB=0.01}, B=NamedRerouteUsage()), B=ComponentMask(None=NamedRerouteUsage()) {b=true})
  n22 = ScalarParameter() {defaultValue=2, group=Mobile Sand, parameterName=Mobile Sand WaterLine Mask}
  NamedRerouteDeclaration(None=Saturate(None=Multiply(A=@n21, B=@n22) {constB=0.5})) {name=SandOceanMask, nodeColor=<rgba 0 0.1094 1 1>}
  NamedRerouteDeclaration(None=FunctionInput(Preview=Constant()) {bUsePreviewValueAsDefault=true, inputName=TerrainMasks_QuadC, inputType=FunctionInput_Vector4, sortPriority=-200}) {name=TerrainMask_QuadC, nodeColor=<rgba 1 0 0.375 1>}
  NamedRerouteDeclaration(None=@n3.A) {name=HeightMapFromTextures, nodeColor=<rgba 0.5 0 1 1>}
 ]}
